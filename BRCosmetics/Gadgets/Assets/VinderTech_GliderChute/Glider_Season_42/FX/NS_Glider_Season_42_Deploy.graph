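# Niagara deploy effect for the Season 42 glider: a multicolor flare with swirling voxel cubes and glitch blocks.
# Cubes_Pulse bursts cubes outward and Cubes_Wave streams them from a growing ring, each with glow sprites colored from an indexed palette ramp.
# The heroic variant adds color flicker and wider palette picks, and HeroicSpawn gates Glitch_Blocks and an extra flare burst.
# Glow alpha differs in the front end; voxel smoke is disabled.
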NiagaraSystem /BRCosmetics/Gadgets/Assets/VinderTech_GliderChute/Glider_Season_42/FX/NS_Glider_Season_42_Deploy
dependencies {
 usedDataInterfaces [
  &NiagaraDataInterfaceColorCurve
  &NiagaraDataInterfaceCurve
  &NiagaraDataInterfaceVectorCurve
  &NiagaraDataInterfaceArrayFloat3
  &NiagaraDataInterfaceArrayFloat
 ]
 usedDynamicInputs [&RandomRangeVector2D]
 usedModules [
  &FN_VariantManager
  &self:SystemSpawnScript.SystemScriptSource.SystemScriptGraph.NiagaraNodeAssignment_0.SetVariables_#1
  &self:SystemSpawnScript.SystemScriptSource.SystemScriptGraph.NiagaraNodeAssignment_1.SetVariables_#2
  &SystemState
  &Timeline
 ]
}
emitters {
 Cubes_Pulse {
  emitterUpdateScript [
   &EmitterState(Life Cycle Mode=::NewEnumerator1, Loop Behavior=::NewEnumerator1, Loop Delay=1, Loop Duration=1, UseLoopDelay=true)
   &SpawnBurst_Instantaneous(Spawn Count=30, Use Spawn Probability=false, UseLoopCountLimit=false) {module=SpawnBurst_Instantaneous001}
  ]
  particleSpawnScript [
   &InitializeParticle(
    Color Minimum=<rgba 0.8 0.8 0.8 1>
    Color Mode=::NewEnumerator2
    Lifetime=&RandomRangeFloat(Minimum=0.2, Maximum=0.6)
    Mesh Scale Mode=::NewEnumerator3
    Mesh Uniform Scale=&RandomRangeFloat(Minimum=0.03, Maximum=0.035)
    UsePositionOffset=false
    Write Material Random=true
   )
   &ShapeLocation(Handle Radius=3, Large Radius=5, MaskSpawnGroup=false, Override Seed=false, Shape Primitive=::NewEnumerator3, UseOwnerScale=true)
   &AddVelocity(
    Module.Use Falloff Distance=false
    Override Seed=false
    Velocity Mode=::NewEnumerator1
    Velocity Speed=&RandomRangeFloat(Minimum=450, Maximum=550)
   )
   &InitialMeshOrientation(Mesh Orientation Mode=::NewEnumerator2, Orientation Coordinate Space=::Local, Use Rotation Vector=false)
  ]
  particleUpdateScript [
   &ParticleState()
   &ScaleMeshSize(Scale Factor=&VectorFromFloat(Value=&Multiply_Float(
      A=&FloatFromCurve(
       FloatCurve=DataInterface() {propertyValues {Curve {
          defaultValue=1
          keys [
           {arriveTangent=15.6766, interpMode=RCIM_Cubic, leaveTangent=15.6766, tangentMode=RCTM_User, time=0, value=0}
           {
            arriveTangent=0.0339
            arriveTangentWeight=0.5121
            interpMode=RCIM_Cubic
            leaveTangent=0.0339
            leaveTangentWeight=0.5121
            tangentMode=RCTM_User
            time=0.1085
            value=1
           }
          ]
         }}}
       CurveIndex=$Particles.Age
      )
      B=&FloatFromCurve(FloatCurve=DataInterface() {propertyValues {Curve {
          defaultValue=1
          keys [
           {
            arriveTangent=0.0339
            arriveTangentWeight=0.5121
            interpMode=RCIM_Cubic
            leaveTangent=0.0339
            leaveTangentWeight=0.5121
            tangentMode=RCTM_User
            time=0.3149
            value=1
           }
           {arriveTangent=-0.324, interpMode=RCIM_Cubic, leaveTangent=-0.324, tangentMode=RCTM_User, time=0.657, value=0.8264}
           {arriveTangent=-4.4271, interpMode=RCIM_Cubic, leaveTangent=-4.4271, tangentMode=RCTM_User, time=1, value=0}
          ]
         }}})
     )))
   &Drag(Drag=4.5, Rotational Drag=4.5, Use Linear Drag=true, Use Rotational Drag=true)
   &VortexForce(UseOriginOffset=false, Vortex Force Amount=&RandomRangeFloat(Minimum=550, Maximum=650))
   /SetVariables(Particles.ColorIndexFloat:NiagaraFloat=/Niagara/DynamicInputs/Math/Floor_Float(Float:NiagaraFloat=&MakeCustomFloatFromBool(
      Bool=$System.isHeroic
      True Float=&Add_Float(
       A=&RandomRangeFloat(Minimum=0, Maximum=16)
       B=&MakeCustomFloatFromBool(
        Bool=&RandomBool()
        True Float=&FloatFromCurve(
         FloatCurve=DataInterface() {propertyValues {Curve {
            defaultValue=1
            keys [
             <at 0 0 RCIM_Cubic>
             {arriveTangent=-0.1076, interpMode=RCIM_Cubic, leaveTangent=-0.1076, tangentMode=RCTM_User, time=0.5041, value=0.4853}
             <at 1 1 RCIM_Cubic>
            ]
           }}}
         Scale Curve=&RandomRangeFloat(Minimum=2, Maximum=4)
        )
        False Float=&RandomRangeFloat(Minimum=0, Maximum=6, Evaluation Type=::NewEnumerator1)
       )
      )
      False Float=&RandomRangeFloat(Minimum=0, Maximum=12)
     ))) {isSetParameters=true}
   &SolveForcesAndVelocity(Clamp Velocity=false, Limit Acceleration=false)
   /SetVariables(Particles.VoxelParams:Vector4f=&MakeVector4(X=&RandomRangeFloat(Minimum=0.4, Maximum=0.8), Y=0, Z=0, W=0)) {isSetParameters=true}
   /SetVariables(
    Particles.Glow_Base_CameraOffset=-5
    Particles.Glow_Base_Color:LinearColor=&MakeLinearColorFromVectorAndFloat(
     "Vector (RGB)"=&MakeVectorFromLinearColorRGB(Color=&ColorFromCurve(
       ColorCurve=DataInterface() {propertyValues {
         AlphaCurve {keys [<at 0.5 1 RCIM_Cubic>]}
         BlueCurve {keys [
           {interpMode=RCIM_Cubic, leaveTangentWeight=0.0208, time=0, value=0.7605}
           {
            arriveTangent=-0.9834
            arriveTangentWeight=0.0208
            interpMode=RCIM_Cubic
            leaveTangent=-0.9834
            leaveTangentWeight=0.0625
            time=0.0312
            value=0.6376
           }
           {
            arriveTangent=1.0336
            arriveTangentWeight=0.0625
            interpMode=RCIM_Cubic
            leaveTangent=1.0336
            leaveTangentWeight=0.0208
            time=0.125
            value=0.6376
           }
           {arriveTangent=0.5926, interpMode=RCIM_Cubic, leaveTangent=0.5926, time=0.1875, value=0.7991}
           {arriveTangent=-2.2224, interpMode=RCIM_Cubic, leaveTangent=-2.2224, time=0.4375, value=0.8228}
           {arriveTangent=-5.133, interpMode=RCIM_Cubic, leaveTangent=-5.133, time=0.5, value=0.1046}
           {arriveTangent=-0.2282, interpMode=RCIM_Cubic, leaveTangent=-0.2282, time=0.5625, value=0.1812}
           {
            arriveTangent=-0.4342
            arriveTangentWeight=0.0279
            interpMode=RCIM_Cubic
            leaveTangent=-0.4342
            leaveTangentWeight=0.0759
            time=0.875
            value=0.0015
           }
           {arriveTangent=-0.4342, interpMode=RCIM_Cubic, leaveTangent=-0.4342, time=0.9375, value=0.0048}
           {arriveTangent=12.092, interpMode=RCIM_Cubic, leaveTangent=12.092, time=0.9688, value=0.0048}
           <at 1 0.7605 RCIM_Cubic>
          ]}
         GreenCurve {keys [
           {interpMode=RCIM_Cubic, leaveTangentWeight=0.0208, time=0, value=0.1441}
           {
            arriveTangent=-0.8979
            arriveTangentWeight=0.0208
            interpMode=RCIM_Cubic
            leaveTangent=-0.8979
            leaveTangentWeight=0.0625
            time=0.0312
            value=0.0319
           }
           {arriveTangent=2.0457, interpMode=RCIM_Cubic, leaveTangent=2.0457, time=0.125, value=0.0319}
           {arriveTangent=2.2592, interpMode=RCIM_Cubic, leaveTangent=2.2592, time=0.1875, value=0.3515}
           {arriveTangent=1.3088, interpMode=RCIM_Cubic, leaveTangent=1.3088, time=0.4375, value=0.7379}
           {arriveTangent=0.365, interpMode=RCIM_Cubic, leaveTangent=0.365, time=0.5, value=0.7605}
           {arriveTangent=-1.555, interpMode=RCIM_Cubic, leaveTangent=-1.555, time=0.5625, value=0.7835}
           {
            arriveTangent=-1.7312
            arriveTangentWeight=0.0671
            interpMode=RCIM_Cubic
            leaveTangent=-1.7312
            leaveTangentWeight=0.1825
            time=0.875
            value=0.3663
           }
           {arriveTangent=-1.7312, interpMode=RCIM_Cubic, leaveTangent=-1.7312, time=0.9375, value=0.0802}
           {arriveTangent=1.0225, interpMode=RCIM_Cubic, leaveTangent=1.0225, time=0.9688, value=0.0802}
           <at 1 0.1441 RCIM_Cubic>
          ]}
         RedCurve {keys [
           {interpMode=RCIM_Cubic, leaveTangentWeight=0.0208, time=0, value=0.6867}
           {arriveTangent=-5.132, interpMode=RCIM_Cubic, leaveTangent=-5.132, time=0.0312, value=0.1119}
           {arriveTangent=-0.5198, interpMode=RCIM_Cubic, leaveTangent=-0.5198, time=0.125, value=0.0452}
           {arriveTangent=0.1677, interpMode=RCIM_Cubic, leaveTangent=0.1677, time=0.1875, value=0.0307}
           {arriveTangent=0.4007, interpMode=RCIM_Cubic, leaveTangent=0.4007, time=0.4375, value=0.0976}
           {arriveTangent=5.4259, interpMode=RCIM_Cubic, leaveTangent=5.4259, time=0.5, value=0.1559}
           {arriveTangent=1.5243, interpMode=RCIM_Cubic, leaveTangent=1.5243, time=0.5625, value=0.7758}
           {
            arriveTangent=0.1156
            arriveTangentWeight=0.0234
            interpMode=RCIM_Cubic
            leaveTangent=0.1156
            leaveTangentWeight=0.0635
            time=0.875
            value=0.807
           }
           {arriveTangent=0.1156, interpMode=RCIM_Cubic, leaveTangent=0.1156, time=0.9375, value=0.8228}
           {arriveTangent=-2.1776, interpMode=RCIM_Cubic, leaveTangent=-2.1776, time=0.9688, value=0.8228}
           <at 1 0.6867 RCIM_Cubic>
          ]}
        }}
       CurveIndex=&Frac_Float(Float=&Divide_Float(A=$Particles.ColorIndexFloat, B=16))
      ))
     "Float (Alpha)"=&FN_FrontEndFloatModifier(Gameplay=0.3, Front End=0.15)
    )
    Particles.Glow_Base_Size:Vector2f=&Vector2DFromFloat(Value=&Multiply_Float(A=&MakeFloatFromVector(Vector=$Particles.Scale, Channel=::NewEnumerator0), B=1230))
   ) {isSetParameters=true}
   &ScaleColor(
    Curve Index=&MakeCustomFloatFromBool(
     Bool=$System.isHeroic
     True Float=&Add_Float(A=$Particles.NormalizedAge, B=&RandomRangeFloat(Minimum=-0.1, Maximum=0.1))
     False Float=1
    )
    Linear Color Curve=DataInterface() {propertyValues {
      AlphaCurve {keys [<at 0 1 RCIM_Constant>]}
      BlueCurve {keys [<at 0.0829 5 RCIM_Constant>,<at 0.1708 0.4 RCIM_Constant>,<at 0.2162 1.2615 RCIM_Constant>,<at 0.2871 2 RCIM_Constant>,<at 0.3396 1.1321 RCIM_Constant>,<at 0.3551 0.7 RCIM_Constant>,<at 0.4189 1 RCIM_Constant>]}
      GreenCurve {keys [<at 0.0829 5 RCIM_Constant>,<at 0.1708 0.4 RCIM_Constant>,<at 0.2162 1.2615 RCIM_Constant>,<at 0.2871 2 RCIM_Constant>,<at 0.3396 1.1321 RCIM_Constant>,<at 0.3551 0.7 RCIM_Constant>,<at 0.4189 1 RCIM_Constant>]}
      RedCurve {keys [<at 0.0829 5 RCIM_Constant>,<at 0.1708 0.4 RCIM_Constant>,<at 0.2162 1.2615 RCIM_Constant>,<at 0.2871 2 RCIM_Constant>,<at 0.3396 1.1321 RCIM_Constant>,<at 0.3551 0.7 RCIM_Constant>,<at 0.4189 1 RCIM_Constant>]}
     }}
    Scale Mode=2
   )
  ]
  rendererClasses [&NiagaraMeshRendererProperties, &NiagaraSpriteRendererProperties]
  renderers [
   {rendererClass=&NiagaraMeshRendererProperties, rendererIndex=0}
   {rendererClass=&NiagaraSpriteRendererProperties, rendererIndex=1}
  ]
 }
 Cubes_Wave {
  emitterUpdateScript [
   &EmitterState(Life Cycle Mode=::NewEnumerator1, Loop Behavior=::NewEnumerator1, Loop Delay=1, Loop Duration=0.5, UseLoopDelay=true)
   &SpawnRate(
    SpawnRate=&FloatFromCurve(
     FloatCurve=DataInterface() {propertyValues {Curve {keys [{arriveTangent=1.2787, interpMode=RCIM_Cubic, leaveTangent=1.2787, tangentMode=RCTM_User, time=0, value=0.4118}, <at 1 1 RCIM_Cubic>]}}}
     CurveIndex=$Emitter.NormalizedLoopAge
     Scale Curve=60
    )
    Use Spawn Probability=false
   )
  ]
  particleSpawnScript [
   &InitializeParticle(
    Color Maximum=<rgba 0.8 0.8 0.8 1>
    Color Minimum=<rgba 0.4 0.4 0.4 1>
    Color Mode=::NewEnumerator2
    Lifetime=&RandomRangeFloat(Minimum=0.3, Maximum=0.7)
    Mesh Scale Mode=::NewEnumerator3
    Mesh Uniform Scale=&RandomRangeFloat(Minimum=0.02, Maximum=0.025)
    UsePositionOffset=false
    Write Material Random=true
   )
   &ShapeLocation(
    Handle Radius=6
    Large Radius=&FloatFromCurve(
     FloatCurve=DataInterface() {propertyValues {Curve {keys [{arriveTangent=2.0697, interpMode=RCIM_Cubic, leaveTangent=2.0697, tangentMode=RCTM_User, time=0, value=0.1529}, <at 1 1 RCIM_Cubic>]}}}
     CurveIndex=$Emitter.NormalizedLoopAge
     Scale Curve=90
    )
    MaskSpawnGroup=false
    Override Seed=false
    Shape Primitive=::NewEnumerator3
    UseOwnerScale=true
   )
   &AddVelocity(
    Module.Use Falloff Distance=false
    Override Seed=false
    Velocity Mode=::NewEnumerator1
    Velocity Speed=&RandomRangeFloat(Minimum=50, Maximum=90)
   ) {module=AddVelocity001}
   &AddVelocity(Velocity=&RandomVector(Vector Scale=1), Velocity Speed Scale=70)
   &InitialMeshOrientation(Mesh Orientation Mode=::NewEnumerator2, Orientation Coordinate Space=::Local, Use Rotation Vector=false)
  ]
  particleUpdateScript [
   &ParticleState()
   &ScaleMeshSize(Scale Factor=&VectorFromFloat(Value=&FloatFromCurve(FloatCurve=DataInterface() {propertyValues {Curve {
         defaultValue=1
         keys [
          {arriveTangent=2.6725, interpMode=RCIM_Cubic, leaveTangent=2.6725, tangentMode=RCTM_User, time=0, value=0}
          {
           arriveTangent=0.0339
           arriveTangentWeight=0.5121
           interpMode=RCIM_Cubic
           leaveTangent=0.0339
           leaveTangentWeight=0.5121
           tangentMode=RCTM_User
           time=0.3635
           value=0.9412
          }
          {arriveTangent=-0.324, interpMode=RCIM_Cubic, leaveTangent=-0.324, tangentMode=RCTM_User, time=0.5046, value=0.9324}
          {arriveTangent=-1.8821, interpMode=RCIM_Cubic, leaveTangent=-1.8821, time=0.7523, value=0.7559}
          {arriveTangent=-4.5392, interpMode=RCIM_Cubic, leaveTangent=-4.5392, tangentMode=RCTM_User, time=1, value=0}
         ]
        }}})))
   &Drag(Drag=4.5, Rotational Drag=4.5, Use Linear Drag=true, Use Rotational Drag=true)
   &VortexForce(UseOriginOffset=false, Vortex Force Amount=&RandomRangeFloat(Minimum=150, Maximum=220))
   /SetVariables(Particles.ColorIndexFloat:NiagaraFloat=&Floor_Float(Float:NiagaraFloat=&MakeCustomFloatFromBool(
      Bool=$System.isHeroic
      True Float=&Add_Float(
       A=&RandomRangeFloat(Minimum=0, Maximum=16)
       B=&MakeCustomFloatFromBool(
        Bool=&RandomBool()
        True Float=&FloatFromCurve(
         FloatCurve=DataInterface() {propertyValues {Curve {
            defaultValue=1
            keys [
             <at 0 0 RCIM_Cubic>
             {arriveTangent=-0.1076, interpMode=RCIM_Cubic, leaveTangent=-0.1076, tangentMode=RCTM_User, time=0.5041, value=0.4853}
             <at 1 1 RCIM_Cubic>
            ]
           }}}
         Scale Curve=&RandomRangeFloat(Minimum=0, Maximum=12)
        )
        False Float=0
       )
      )
      False Float=&RandomRangeFloat(Minimum=0, Maximum=12)
     ))) {isSetParameters=true}
   &SolveForcesAndVelocity(Clamp Velocity=false, Limit Acceleration=false)
   /SetVariables(
    Particles.VoxelParams:Vector4f=&MakeVector4(X=&RandomRangeFloat(Minimum=0.4, Maximum=0.8), Y=0, Z=0, W=0)
    Particles.VoxelSubUVIndex:NiagaraFloat=&Multiply_Float(A=&Frac_Float(Float=$Particles.ColorIndexFloat), B=16)
   ) {isSetParameters=true}
   /SetVariables(
    Particles.Glow_Base_CameraOffset=-5
    Particles.Glow_Base_Color:LinearColor=&MakeLinearColorFromVectorAndFloat(
     "Vector (RGB)"=&MakeVectorFromLinearColorRGB(Color=&ColorFromCurve(
       ColorCurve=DataInterface() {propertyValues {
         AlphaCurve {keys [<at 0.5 1 RCIM_Cubic>]}
         BlueCurve {keys [
           {interpMode=RCIM_Cubic, leaveTangentWeight=0.0208, time=0, value=0.7605}
           {
            arriveTangent=-0.9834
            arriveTangentWeight=0.0208
            interpMode=RCIM_Cubic
            leaveTangent=-0.9834
            leaveTangentWeight=0.0625
            time=0.0312
            value=0.6376
           }
           {
            arriveTangent=1.0336
            arriveTangentWeight=0.0625
            interpMode=RCIM_Cubic
            leaveTangent=1.0336
            leaveTangentWeight=0.0208
            time=0.125
            value=0.6376
           }
           {arriveTangent=0.5926, interpMode=RCIM_Cubic, leaveTangent=0.5926, time=0.1875, value=0.7991}
           {arriveTangent=-2.2224, interpMode=RCIM_Cubic, leaveTangent=-2.2224, time=0.4375, value=0.8228}
           {arriveTangent=-5.133, interpMode=RCIM_Cubic, leaveTangent=-5.133, time=0.5, value=0.1046}
           {arriveTangent=-0.2282, interpMode=RCIM_Cubic, leaveTangent=-0.2282, time=0.5625, value=0.1812}
           {
            arriveTangent=-0.4342
            arriveTangentWeight=0.0279
            interpMode=RCIM_Cubic
            leaveTangent=-0.4342
            leaveTangentWeight=0.0759
            time=0.875
            value=0.0015
           }
           {arriveTangent=-0.4342, interpMode=RCIM_Cubic, leaveTangent=-0.4342, time=0.9375, value=0.0048}
           {arriveTangent=12.092, interpMode=RCIM_Cubic, leaveTangent=12.092, time=0.9688, value=0.0048}
           <at 1 0.7605 RCIM_Cubic>
          ]}
         GreenCurve {keys [
           {interpMode=RCIM_Cubic, leaveTangentWeight=0.0208, time=0, value=0.1441}
           {
            arriveTangent=-0.8979
            arriveTangentWeight=0.0208
            interpMode=RCIM_Cubic
            leaveTangent=-0.8979
            leaveTangentWeight=0.0625
            time=0.0312
            value=0.0319
           }
           {arriveTangent=2.0457, interpMode=RCIM_Cubic, leaveTangent=2.0457, time=0.125, value=0.0319}
           {arriveTangent=2.2592, interpMode=RCIM_Cubic, leaveTangent=2.2592, time=0.1875, value=0.3515}
           {arriveTangent=1.3088, interpMode=RCIM_Cubic, leaveTangent=1.3088, time=0.4375, value=0.7379}
           {arriveTangent=0.365, interpMode=RCIM_Cubic, leaveTangent=0.365, time=0.5, value=0.7605}
           {arriveTangent=-1.555, interpMode=RCIM_Cubic, leaveTangent=-1.555, time=0.5625, value=0.7835}
           {
            arriveTangent=-1.7312
            arriveTangentWeight=0.0671
            interpMode=RCIM_Cubic
            leaveTangent=-1.7312
            leaveTangentWeight=0.1825
            time=0.875
            value=0.3663
           }
           {arriveTangent=-1.7312, interpMode=RCIM_Cubic, leaveTangent=-1.7312, time=0.9375, value=0.0802}
           {arriveTangent=1.0225, interpMode=RCIM_Cubic, leaveTangent=1.0225, time=0.9688, value=0.0802}
           <at 1 0.1441 RCIM_Cubic>
          ]}
         RedCurve {keys [
           {interpMode=RCIM_Cubic, leaveTangentWeight=0.0208, time=0, value=0.6867}
           {arriveTangent=-5.132, interpMode=RCIM_Cubic, leaveTangent=-5.132, time=0.0312, value=0.1119}
           {arriveTangent=-0.5198, interpMode=RCIM_Cubic, leaveTangent=-0.5198, time=0.125, value=0.0452}
           {arriveTangent=0.1677, interpMode=RCIM_Cubic, leaveTangent=0.1677, time=0.1875, value=0.0307}
           {arriveTangent=0.4007, interpMode=RCIM_Cubic, leaveTangent=0.4007, time=0.4375, value=0.0976}
           {arriveTangent=5.4259, interpMode=RCIM_Cubic, leaveTangent=5.4259, time=0.5, value=0.1559}
           {arriveTangent=1.5243, interpMode=RCIM_Cubic, leaveTangent=1.5243, time=0.5625, value=0.7758}
           {
            arriveTangent=0.1156
            arriveTangentWeight=0.0234
            interpMode=RCIM_Cubic
            leaveTangent=0.1156
            leaveTangentWeight=0.0635
            time=0.875
            value=0.807
           }
           {arriveTangent=0.1156, interpMode=RCIM_Cubic, leaveTangent=0.1156, time=0.9375, value=0.8228}
           {arriveTangent=-2.1776, interpMode=RCIM_Cubic, leaveTangent=-2.1776, time=0.9688, value=0.8228}
           <at 1 0.6867 RCIM_Cubic>
          ]}
        }}
       CurveIndex=&Frac_Float(Float=&Divide_Float(A=$Particles.ColorIndexFloat, B=16))
      ))
     "Float (Alpha)"=&FN_FrontEndFloatModifier(Gameplay=0.3, Front End=0.15)
    )
    Particles.Glow_Base_Size:Vector2f=&Vector2DFromFloat(Value=&Multiply_Float(A=&MakeFloatFromVector(Vector=$Particles.Scale, Channel=::NewEnumerator0), B=1230))
   ) {isSetParameters=true}
   &ScaleColor(
    Curve Index=&MakeCustomFloatFromBool(
     Bool=$System.isHeroic
     True Float=&Add_Float(A=$Particles.NormalizedAge, B=&RandomRangeFloat(Minimum=-0.1, Maximum=0.1))
     False Float=1
    )
    Linear Color Curve=DataInterface() {propertyValues {
      AlphaCurve {keys [<at 0 1 RCIM_Constant>]}
      BlueCurve {keys [<at 0.0829 5 RCIM_Constant>,<at 0.1708 0.6 RCIM_Constant>,<at 0.2162 1.2615 RCIM_Constant>,<at 0.2871 2 RCIM_Constant>,<at 0.3396 1.1321 RCIM_Constant>,<at 0.3551 0.4 RCIM_Constant>,<at 0.4189 1 RCIM_Constant>]}
      GreenCurve {keys [<at 0.0829 5 RCIM_Constant>,<at 0.1708 0.6 RCIM_Constant>,<at 0.2162 1.2615 RCIM_Constant>,<at 0.2871 2 RCIM_Constant>,<at 0.3396 1.1321 RCIM_Constant>,<at 0.3551 0.4 RCIM_Constant>,<at 0.4189 1 RCIM_Constant>]}
      RedCurve {keys [<at 0.0829 5 RCIM_Constant>,<at 0.1708 0.6 RCIM_Constant>,<at 0.2162 1.2615 RCIM_Constant>,<at 0.2871 2 RCIM_Constant>,<at 0.3396 1.1321 RCIM_Constant>,<at 0.3551 0.4 RCIM_Constant>,<at 0.4189 1 RCIM_Constant>]}
     }}
    Scale Mode=2
   )
   /SetVariables(Particles.JitterPosition:NiagaraPosition=&Add_Vector(
     A=Linked() {linkedVariable {name=Particles.Position, type=NiagaraPosition}}
     B=&Multiply_VectorByFloat(
      Vector=&RandomRangeVector(Minimum=<-2 -2 -2>, Maximum=<2 2 2>, Evaluation Type=::NewEnumerator1)
      Float=&FloatFromCurve(
       FloatCurve=DataInterface() {propertyValues {Curve {keys [<at 0 0 RCIM_Constant>,<at 0.8991 1 RCIM_Constant>]}}}
       CurveIndex=&RandomRangeFloat(Minimum=0, Maximum=1, Evaluation Type=::NewEnumerator1)
       Scale Curve=&FloatFromCurve(FloatCurve=DataInterface() {propertyValues {Curve {
           defaultValue=0
           keys [
            {interpMode=RCIM_Cubic, tangentMode=RCTM_User, time=0, value=0}
            {interpMode=RCIM_Cubic, tangentMode=RCTM_User, time=0.5, value=1}
            {tangentMode=RCTM_User, time=1, value=0}
           ]
          }}})
      )
     )
    )) {isSetParameters=true}
  ]
  rendererClasses [&NiagaraMeshRendererProperties, &NiagaraSpriteRendererProperties]
  renderers [
   {rendererClass=&NiagaraMeshRendererProperties, rendererIndex=0}
   {rendererClass=&NiagaraSpriteRendererProperties, rendererIndex=1}
  ]
 }
 Flare {
  emitterUpdateScript [
   &EmitterState(Life Cycle Mode=::NewEnumerator1, Loop Behavior=::NewEnumerator1, Loop Delay=1, Loop Duration=0.2, UseLoopDelay=true)
   &SpawnBurst_Instantaneous(Use Spawn Probability=false, UseLoopCountLimit=false)
   &SpawnBurst_Instantaneous(Spawn Group=1, Spawn Probability=$System.HeroicSpawn, Use Spawn Probability=true, UseLoopCountLimit=false) {module=SpawnBurst_Instantaneous001}
  ]
  particleSpawnScript [
   /SetVariables(Particles.VisibilityTag=&SpawnGroup()) {isSetParameters=true}
   &InitializeParticle(Lifetime=0.57, Sprite Size Mode=::NewEnumerator3, Uniform Sprite Size=380, UsePositionOffset=true, Write Material Random=true)
   &CameraOffset(Camera Offset Amount=5)
  ]
  particleUpdateScript [
   &ParticleState()
   &ScaleColor(
    Scale Alpha=&FloatFromCurve(
     FloatCurve=DataInterface() {propertyValues {Curve {
        defaultValue=1
        keys [
         {arriveTangent=9.0725, interpMode=RCIM_Cubic, leaveTangent=9.0725, tangentMode=RCTM_User, time=0, value=0}
         {arriveTangent=-2.5249, interpMode=RCIM_Cubic, leaveTangent=-2.5249, tangentMode=RCTM_User, time=0.1408, value=1}
         {arriveTangent=-0.0405, interpMode=RCIM_Cubic, leaveTangent=-0.0405, tangentMode=RCTM_User, time=1, value=0}
        ]
       }}}
     Scale Curve=2
    )
    Scale RGB=&MakeVectorFromLinearColorRGB(Color=&ColorFromCurve(ColorCurve=DataInterface() {propertyValues {
        AlphaCurve {keys [<at 0 1 RCIM_Cubic>,<at 1 0 RCIM_Cubic>]}
        BlueCurve {keys [<at 0 1>,<at 0.1745 1 RCIM_Cubic>,<at 0.2324 1>,<at 0.3483 0.6147>,<at 0.5087 0.087 RCIM_Cubic>]}
        GreenCurve {keys [
          <at 0 0.8226>
          {arriveTangent=-2.1549, interpMode=RCIM_Cubic, leaveTangent=-2.1549, time=0.1745, value=0.7827}
          <at 0.2324 0.3217>
          <at 0.3483 0.287>
          <at 0.5087 0.4989 RCIM_Cubic>
         ]}
        RedCurve {keys [
          <at 0 0.2261>
          {arriveTangent=0.727, interpMode=RCIM_Cubic, leaveTangent=0.727, time=0.1745, value=0.0522}
          <at 0.2324 0.3951>
          <at 0.3483 1>
          <at 0.5087 1 RCIM_Cubic>
         ]}
       }}))
    ScaleA=true
    ScaleRGB=true
   )
   &SolveForcesAndVelocity(Clamp Velocity=false, Limit Acceleration=false)
   /SetVariables(
    Particles.CoronaSpriteColor:LinearColor=&MultiplyLinearColor(Linear Color=$Particles.Color, Scale Factor=<1 1 1 0.5>)
    Particles.CoronaSpriteSize=&Vector2DFromFloat(Value=&FloatFromCurve(
      FloatCurve=DataInterface() {propertyValues {Curve {
         defaultValue=0
         keys [
          {arriveTangent=0.5164, interpMode=RCIM_Cubic, leaveTangent=0.5164, tangentMode=RCTM_User, time=0, value=1}
          {interpMode=RCIM_Cubic, tangentMode=RCTM_User, time=1, value=1.2}
         ]
        }}}
      Scale Curve=175
     ))
   ) {isSetParameters=true}
   /SetVariables(
    Particles.Flare Dyn Param 0:Vector4f=&MakeVector4(
     X=&FloatFromCurve(FloatCurve=DataInterface() {propertyValues {Curve {keys [<at 0 0.6 RCIM_Cubic>,<at 0.8 0 RCIM_Cubic>]}}})
     Y=2
     Z=1
     W=1
    )
    Particles.Flare Dyn Param 1:Vector4f=&MakeVector4(X=0, Y=0.4, Z=0.6, W=1)
   ) {isSetParameters=true}
   /SetVariables(
    Particles.PixelHaloColor:LinearColor=&ColorFromCurve(ColorCurve=DataInterface() {propertyValues {
       AlphaCurve {keys [
         <at 0.0051 0>
         {arriveTangent=0.203, interpMode=RCIM_Cubic, leaveTangent=0.203, time=0.1313, value=1}
         <at 0.272 0.4078>
         <at 0.4258 0.0854>
         <at 0.5549 0 RCIM_Cubic>
        ]}
       BlueCurve {keys [<at 0.1342 2 RCIM_Cubic>,<at 0.2981 1>,<at 0.433 1>]}
       GreenCurve {keys [<at 0.1342 1.4891 RCIM_Cubic>,<at 0.2981 0.2545>,<at 0.433 0.2545>]}
       RedCurve {keys [<at 0.1342 0.2154 RCIM_Cubic>,<at 0.2981 0.0435>,<at 0.433 0.0435>]}
      }})
    Particles.PixelHaloParams:Vector4f=<0.0939 1.4496 0 19.1887>
    Particles.PixelHaloRotation=135
    Particles.PixelHaloSize:Vector2f=&Vector2DFromFloat(Value=&FloatFromCurve(
      FloatCurve=DataInterface() {propertyValues {Curve {
         defaultValue=1
         keys [
          {arriveTangent=4.1998, interpMode=RCIM_Constant, time=0.0035, value=0.349}
          <at 0.0979 0.7707 RCIM_Constant>
          <at 0.2055 0.9103 RCIM_Constant>
          <at 0.3253 0.6543 RCIM_Constant>
          {arriveTangent=-0.1205, interpMode=RCIM_Constant, leaveTangent=-0.1205, time=0.4115, value=0.2634}
         ]
        }}}
      Scale Curve=130.7718
     ))
   ) {isSetParameters=true}
  ]
  rendererClasses [&NiagaraSpriteRendererProperties]
  renderers [
   {rendererClass=&NiagaraSpriteRendererProperties}
   {rendererClass=&NiagaraSpriteRendererProperties, rendererIndex=1}
   {rendererClass=&NiagaraSpriteRendererProperties, rendererIndex=2}
  ]
 }
 Glitch_Blocks {
  emitterUpdateScript [
   &EmitterState(Life Cycle Mode=::NewEnumerator1, Loop Behavior=::NewEnumerator1, Loop Delay=1, Loop Duration=0.45, UseLoopDelay=true)
   &SpawnRate(Spawn Probability=$System.HeroicSpawn, SpawnRate=60, Use Spawn Probability=true)
  ]
  particleSpawnScript [
   /SetVariables(Particles.VisibilityTag=&MakeCustomIntFromBool(
     Bool=&SetBoolByFloatComparison(A=&RandomRangeFloat(Minimum=0, Maximum=1), B=0.1)
     True INT=1
     False INT=0
    )) {isSetParameters=true}
   &InitializeParticle(
    Color=&MaskLinearColor(Mask Source=1, Mask Parameter=$Particles.VisibilityTag, Mask=1, Masked Linear Color=<rgba 3 3 3 9>, Passthrough Linear Color=<rgba 1 1 1 2>)
    Lifetime=&RandomRangeFloat(Minimum=0.16, Maximum=0.24)
    Mesh Scale=&VectorFromCurve(
     VectorCurve=DataInterface() {propertyValues {
       XCurve {keys [<at 0 0.5844>,<at 0.1624 0.3351 RCIM_Constant>,<at 0.1957 0.4679>,<at 0.5247 0.3759 RCIM_Constant>,<at 0.5602 1.258>,<at 1 0.8335 RCIM_Constant>]}
       YCurve {keys [<at 0 0.2314 RCIM_Constant>,<at 0.1957 0.8136>,<at 0.5247 0.5854 RCIM_Constant>,<at 0.5602 0.4973>,<at 1 0.4039 RCIM_Constant>]}
       ZCurve {keys [<at 0 0.8813>,<at 0.1624 0.5926 RCIM_Constant>,<at 0.1957 0.2614 RCIM_Constant>,<at 0.5602 0.2614 RCIM_Constant>]}
      }}
     CurveIndex=&RandomRangeFloat(Minimum=0, Maximum=1)
     Scale Curve=<1.4 1.4 1.2>
    )
    Mesh Scale Mode=::NewEnumerator4
    UsePositionOffset=false
    Write Material Random=true
   )
   &CameraOffset()
   &InitialMeshOrientation(
    Mesh Axis To Orient=<0 0 1>
    Mesh Orientation Mode=::NewEnumerator2
    Orientation Coordinate Space=::Local
    Orientation Vector=<0 0 1>
    Use Rotation Vector=false
   )
   &ShapeLocation(
    MaskSpawnGroup=false
    Override Seed=false
    Sphere Radius=&Lerp_Float(A=60, B=90, Alpha=$Emitter.NormalizedLoopAge)
    UseOwnerScale=true
   )
   &AddVelocity(
    Module.Use Falloff Distance=false
    Override Seed=false
    Velocity Mode=::NewEnumerator1
    Velocity Speed=&RandomRangeFloat(Minimum=30, Maximum=90)
   )
  ]
  particleUpdateScript [
   &ParticleState()
   &Drag(Drag=3, Rotational Drag=3, Use Linear Drag=true, Use Rotational Drag=true)
   &SolveForcesAndVelocity(Clamp Velocity=false, Limit Acceleration=false)
   &ScaleColor(
    Scale Alpha=&FloatFromCurve(FloatCurve=DataInterface() {propertyValues {Curve {
        defaultValue=1
        keys [
         {arriveTangent=1.2192, interpMode=RCIM_Cubic, leaveTangent=1.2192, tangentMode=RCTM_User, time=0, value=0}
         {arriveTangent=3.0011, interpMode=RCIM_Cubic, leaveTangent=3.0011, time=0.1697, value=0.4382}
         {interpMode=RCIM_Cubic, tangentMode=RCTM_User, time=0.3332, value=1}
         {arriveTangent=-1.1786, tangentMode=RCTM_User, time=1, value=0}
        ]
       }}})
    Scale RGB=&MakeVectorFromLinearColorRGB(Color=&ColorFromCurve(
      ColorCurve=DataInterface() {propertyValues {
        AlphaCurve {keys [<at 0 1 RCIM_Cubic>]}
        BlueCurve {keys [
          <at 0 0.7523>
          <at 0.0542 1>
          <at 0.1804 1>
          <at 0.3304 1>
          {arriveTangent=-1.5974, interpMode=RCIM_Cubic, leaveTangent=-1.5974, time=0.5413, value=1}
          <at 0.7344 0.3547>
          <at 0.7983 0.1565>
          <at 1 0.1478 RCIM_Cubic>
         ]}
        GreenCurve {keys [
          <at 0 0.1632>
          <at 0.0542 0.0609>
          <at 0.1804 0.087>
          <at 0.3304 0.2463>
          {arriveTangent=1.8657, interpMode=RCIM_Cubic, leaveTangent=1.8657, time=0.5413, value=0.7595}
          <at 0.7344 1>
          <at 0.7983 0.9528>
          <at 1 0.2044 RCIM_Cubic>
         ]}
        RedCurve {keys [
          <at 0 1>
          <at 0.0542 0.4789>
          <at 0.1804 0.3414>
          <at 0.3304 0.1019>
          {arriveTangent=0.3073, interpMode=RCIM_Cubic, leaveTangent=0.3073, time=0.5413, value=0}
          <at 0.7344 0.2261>
          <at 0.7983 1>
          <at 1 1 RCIM_Cubic>
         ]}
       }}
      CurveIndex=&RandomRangeFloat(Minimum=0, Maximum=1, Evaluation Type=::NewEnumerator1)
     ))
    ScaleA=true
    ScaleRGB=true
   )
   &ScaleMeshSize(Scale Factor=&MaskVectorBySpawnGroup(
     Masked Vector=<1 1 1>
     Passthrough Vector=&MakeVector(
      X=&FloatFromCurve(FloatCurve=DataInterface() {propertyValues {Curve {
          defaultValue=1
          keys [
           {arriveTangent=1.4643, time=0, value=0.7528}
           {arriveTangent=0.0248, interpMode=RCIM_Cubic, leaveTangent=0.0248, tangentMode=RCTM_User, time=0.2571, value=1}
           {arriveTangent=-0.187, leaveTangent=-0.187, time=0.7725, value=0.9775}
           {arriveTangent=-0.5952, tangentMode=RCTM_User, time=1, value=0.8611}
          ]
         }}})
      Y=1
      Z=1
     )
     Spawn Group Mask=2
    ))
  ]
  rendererClasses [&NiagaraMeshRendererProperties]
  renderers [
   {rendererClass=&NiagaraMeshRendererProperties, rendererIndex=0}
   {rendererClass=&NiagaraMeshRendererProperties, rendererIndex=1}
  ]
 }
 NE_VoxelsSmoke001 {
  bEnabled=false
  emitterUpdateScript [
   &EmitterState(Life Cycle Mode=::NewEnumerator1, Loop Behavior=::NewEnumerator1, Loop Delay=1, Loop Duration=0.5, UseLoopDelay=true)
   &SpawnBurst_Instantaneous(
    Spawn Count=&RandomRangeInt(Minimum=6, Maximum=9)
    Spawn Probability=$System.HeroicSpawn
    Use Spawn Probability=true
    UseLoopCountLimit=false
   )
  ]
  particleSpawnScript [
   &InitializeParticle(
    Color Mode=::NewEnumerator0
    Lifetime Max=1
    Lifetime Min=0.7
    Lifetime Mode=::NewEnumerator1
    Mesh Scale Mode=::NewEnumerator3
    Mesh Uniform Scale=&RandomRangeFloat(Minimum=0.7, Maximum=0.8)
    Sprite Rotation Mode=::NewEnumerator2
    UsePositionOffset=false
    Write Material Random=true
   )
   &ShapeLocation(MaskSpawnGroup=false, Sphere Distribution=::NewEnumerator3, Sphere Radius=60, UseOwnerScale=true)
   &DynamicMaterialParameters(
    Index 0 Param 1=&RandomRangeFloat(Minimum=0, Maximum=1)
    Index 0 Param 2=1
    Module.Index0Float0Write=true
    Module.Index0Float1Write=true
    Module.Index0Float2Write=true
    Module.Index0Float3Write=true
   ) {enabled=false}
   &InitialMeshOrientation(
    Mesh Orientation Mode=::NewEnumerator2
    Orientation Coordinate Space=::Local
    Orientation Vector=<0 1 0>
    Rotation=&SelectVectorFromArray(
     Array Sampling Mode=::NewEnumerator0
     Direct Array Index=?
     Vector Selection Array=DataInterface() {propertyValues {FloatData [<0 0 0>, <0.25 0 0>, <0.5 0 0>, <0.75 0 0>]}}
     Random Seed=0
    )
    Use Rotation Vector=true
   )
   &InitialMeshOrientation(
    Mesh Orientation Mode=::NewEnumerator1
    Rotation=&SelectVectorFromArray(
     Array Sampling Mode=::NewEnumerator0
     Direct Array Index=?
     Vector Selection Array=DataInterface() {propertyValues {FloatData [<0 0 0.25>, <0.25 0 0.25>, <0.5 0 0.25>, <0.75 0 0>]}}
     Random Seed=0
    )
    Use Rotation Vector=true
   ) {enabled=false, module=InitialMeshOrientation001}
  ]
  particleUpdateScript [
   &Drag(
    Drag=&FloatFromCurve(
     FloatCurve=DataInterface() {propertyValues {Curve {keys [{interpMode=RCIM_Cubic, leaveTangent=3.6589, tangentMode=RCTM_Break, time=0, value=0.15}, <at 0.8 1>]}}}
     Scale Curve=8
    )
    Use Linear Drag=true
    Use Rotational Drag=false
   )
   &GravityForce(Gravity=&Multiply_VectorByFloat(
     Vector=&RandomRangeVector(Minimum=<0 0 50>, Maximum=<0 0 100>)
     Float=&FN_FrontEndFloatModifier(Gameplay=1, Front End=0.4)
    )) {enabled=false}
   &SolveForcesAndVelocity(Clamp Velocity=false, Limit Acceleration=false)
   &ParticleState()
   &ScaleMeshSize(Scale Factor=&VectorFromFloat(Value=&FloatFromCurve(
      FloatCurve=DataInterface() {propertyValues {Curve {keys [
          {interpMode=RCIM_Cubic, leaveTangent=2.4759, tangentMode=RCTM_Break, time=0, value=0}
          {arriveTangent=-0.1759, tangentMode=RCTM_User, time=1, value=1}
         ]}}}
      Scale Curve=0.6
     )))
   &DynamicMaterialParameters(
    Index 0 Param 2=&FloatFromCurve(FloatCurve=DataInterface() {propertyValues {Curve {keys [
         {arriveTangent=2.4806, interpMode=RCIM_Cubic, leaveTangent=2.4806, tangentMode=RCTM_User, time=0, value=0}
         {arriveTangent=1.5524, interpMode=RCIM_Cubic, leaveTangent=1.5524, tangentMode=RCTM_User, time=1, value=1}
        ]}}})
    Module.Index0Float0Write=true
    Module.Index0Float1Write=true
    Module.Index0Float2Write=true
    Module.Index0Float3Write=true
   ) {module=DynamicMaterialParameters001}
   &Color(Color=&LerpLinearColors(
     StartColor=&ColorFromCurve(
      ColorCurve=DataInterface() {propertyValues {
        AlphaCurve {keys [<at 0 1 RCIM_Cubic>,<at 1 0 RCIM_Cubic>]}
        BlueCurve {keys [<at 0.0527 1 RCIM_Cubic>,<at 0.3532 1>,<at 0.6394 0.6006>,<at 0.972 0.1478 RCIM_Cubic>]}
        GreenCurve {keys [<at 0.0527 0.797 RCIM_Cubic>,<at 0.3532 0.3043>,<at 0.6394 0.2609>,<at 0.972 0.5323 RCIM_Cubic>]}
        RedCurve {keys [<at 0.0527 0.1146 RCIM_Cubic>,<at 0.3532 0.3796>,<at 0.6394 1>,<at 0.972 1 RCIM_Cubic>]}
       }}
      CurveIndex=&RandomRangeFloat(Minimum=0, Maximum=1)
     )
     EndColor=<rgba 0.8 0.8 0.8 1>
     LerpFactor=&FloatFromCurve(
      FloatCurve=DataInterface() {propertyValues {Curve {
         defaultValue=1
         keys [
          {arriveTangent=0.2681, interpMode=RCIM_Cubic, leaveTangent=0.2681, tangentMode=RCTM_User, time=0, value=0}
          {arriveTangent=2.2842, interpMode=RCIM_Cubic, leaveTangent=2.2842, time=0.1899, value=0.1676}
          {arriveTangent=1.3156, interpMode=RCIM_Cubic, leaveTangent=1.3156, tangentMode=RCTM_User, time=0.3721, value=0.85}
          {interpMode=RCIM_Cubic, tangentMode=RCTM_User, time=0.79, value=1}
         ]
        }}}
      Scale Curve=0.4
     )
    ))
  ]
  rendererClasses [&NiagaraMeshRendererProperties]
  renderers [{rendererClass=&NiagaraMeshRendererProperties, rendererIndex=0}]
 }
}
summary {
 systemName=NS_Glider_Season_42_Deploy
 emitters [
  {emitterName=Flare, rendererClasses [&NiagaraSpriteRendererProperties]}
  {emitterName=Cubes_Pulse, rendererClasses [&NiagaraMeshRendererProperties, &NiagaraSpriteRendererProperties]}
  {emitterName=Cubes_Wave, rendererClasses [&NiagaraMeshRendererProperties, &NiagaraSpriteRendererProperties]}
  {emitterName=Glitch_Blocks, rendererClasses [&NiagaraMeshRendererProperties]}
  {bEnabled=false, emitterName=NE_VoxelsSmoke001, rendererClasses [&NiagaraMeshRendererProperties]}
 ]
 userVariables [
  User.bIsFrontEnd {type=NiagaraBool}
  User.bIsFrontEndPreview {type=NiagaraBool}
  User.ForwardDot {type=NiagaraFloat, defaultValue=1, description=""}
  User.SizeNormalized {type=NiagaraFloat, defaultValue=1, description=""}
  User.SystemDelay {type=NiagaraFloat, defaultValue=1, description=""}
  User.Variant {type=NiagaraFloat, defaultValue=1, description=""}
 ]
}
