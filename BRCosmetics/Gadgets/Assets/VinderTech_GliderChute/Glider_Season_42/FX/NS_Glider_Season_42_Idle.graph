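# Niagara idle effect for the Season 42 VinderTech glider, combining orbiting cubes, colorful glitch flickers, socket glows and trail cubes.
# CrownCubes circle a sampled skeletal mesh location with jitter and offset glitch copies, while three Glitch emitters add blocks and sprites.
# Glow attaches sprites, meshes and a light to sockets; its Glow001 variant is disabled.
# HeroicSpawn gates most spawning, and TrailCubes also needs TrailSpawn and is suppressed in the front end.
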NiagaraSystem /BRCosmetics/Gadgets/Assets/VinderTech_GliderChute/Glider_Season_42/FX/NS_Glider_Season_42_Idle
dependencies {
 usedDataInterfaces [
  &NiagaraDataInterfaceArrayFloat
  &NiagaraDataInterfaceArrayInt32
  &NiagaraDataInterfaceCurve
  &NiagaraDataInterfaceColorCurve
  &NiagaraDataInterfaceVectorCurve
  &NiagaraDataInterfaceArrayFloat3
  &NiagaraDataInterfaceSkeletalMesh
  &NiagaraDataInterfaceCamera
 ]
 usedDynamicInputs [
  &RandomRangeVector2D
  &MakeCustomFloatFromBool
  &Abs_Float
  &DotProduct
  &ReturnCameraVector
  /Niagara/DynamicInputs/Math/SlerpVector
  /Niagara/DynamicInputs/Vectors/CrossProduct
  &OneMinusFloat
 ]
 usedModules [
  &self:SystemSpawnScript.SystemScriptSource.SystemScriptGraph.NiagaraNodeAssignment_3.SetVariables_#1
  &self:SystemSpawnScript.SystemScriptSource.SystemScriptGraph.NiagaraNodeAssignment_4.SetVariables_#2
  &self:SystemSpawnScript.SystemScriptSource.SystemScriptGraph.NiagaraNodeAssignment_0.SetVariables_#3
  &SystemState
  &BlendSpaceHelper
  &FN_Integrator
  &self:SystemSpawnScript.SystemScriptSource.SystemScriptGraph.NiagaraNodeAssignment_7.SetVariables_#4
  &self:SystemSpawnScript.SystemScriptSource.SystemScriptGraph.NiagaraNodeAssignment_8.SetVariables_#5
  &self:SystemSpawnScript.SystemScriptSource.SystemScriptGraph.NiagaraNodeAssignment_1.SetVariables_#6
 ]
}
emitters {
 CrownCubes {
  emitterUpdateScript [
   &EmitterState(Delay First Loop Only=true, Life Cycle Mode=::NewEnumerator1, Loop Delay=1, Loop Duration=1.4, UseLoopDelay=true)
   &SpawnBurst_Instantaneous(Spawn Count=4, Spawn Probability=$System.HeroicSpawn, Use Spawn Probability=true, UseLoopCountLimit=false)
   /SetVariables(
    Emitter.EmitterHeight:Vector3f=&RandomRangeVector(Minimum=<0 0 -15>, Maximum=<0 0 5>)
    Emitter.EmitterPosition:NiagaraFloat=&Multiply_FloatByInt(Float=1.618, Integer=$Emitter.LoopCount)
    Emitter.EmitterRadius:NiagaraFloat=&RandomRangeFloat(Minimum=0, Maximum=1)
    Emitter.RandomLife:NiagaraFloat=&RandomRangeFloat(Minimum=4, Maximum=5)
   ) {isSetParameters=true}
  ]
  particleSpawnScript [
   &InitializeParticle(
    Color Minimum=<rgba 0.8 0.8 0.8 1>
    Color Mode=::NewEnumerator2
    Lifetime=$Emitter.RandomLife
    Mesh Scale Mode=::NewEnumerator3
    Mesh Uniform Scale=&Subtract_Float(
     A=&Lerp_Float(A=0.025, B=0.0325, Alpha=$Emitter.EmitterRadius)
     B=&Multiply_FloatByInt(Float=0.006, Integer=$Engine.ExecIndex)
    )
    UsePositionOffset=false
    Write Material Random=true
   )
   /SetVariables(
    Particles.HeightRandom:Vector3f=$Emitter.EmitterHeight
    Particles.Order:NiagaraInt32=$Engine.ExecIndex
    Particles.ParticleCirclePosition:NiagaraFloat=$Emitter.EmitterPosition
    Particles.RadiusRandom:NiagaraFloat=$Emitter.EmitterRadius
   ) {isSetParameters=true}
   /SetVariables(Particles.ColorIndexFloat:NiagaraFloat=&Lerp_Float(
     A=12
     B=16
     Alpha=&FloatFromCurve(
      FloatCurve=DataInterface() {propertyValues {Curve {
         defaultValue=1
         keys [
          {arriveTangent=1.9646, interpMode=RCIM_Cubic, leaveTangent=1.9646, tangentMode=RCTM_User, time=0, value=0}
          {arriveTangent=1.8827, interpMode=RCIM_Cubic, leaveTangent=1.8827, tangentMode=RCTM_User, time=1, value=1}
         ]
        }}}
      CurveIndex=&RandomRangeFloat(Minimum=0, Maximum=1)
     )
    )) {isSetParameters=true}
   /SetVariables(Particles.Glow_Base_Color:LinearColor=&MakeLinearColorFromVectorAndFloat(
     "Vector (RGB)"=&MakeVectorFromLinearColorRGB(Color=&ColorFromCurve(
       ColorCurve=DataInterface() {propertyValues {
         AlphaCurve {keys [<at 0.5 90 RCIM_Cubic>]}
         BlueCurve {keys [
           {interpMode=RCIM_Cubic, leaveTangentWeight=0.0208, time=0, value=0.7605}
           {
            arriveTangent=-0.9834
            arriveTangentWeight=0.0208
            interpMode=RCIM_Cubic
            leaveTangent=-0.9834
            leaveTangentWeight=0.0625
            time=0.0312
            value=0.6376
           }
           {
            arriveTangent=1.0336
            arriveTangentWeight=0.0625
            interpMode=RCIM_Cubic
            leaveTangent=1.0336
            leaveTangentWeight=0.0208
            time=0.125
            value=0.6376
           }
           {arriveTangent=0.5926, interpMode=RCIM_Cubic, leaveTangent=0.5926, time=0.1875, value=0.7991}
           {arriveTangent=-2.2224, interpMode=RCIM_Cubic, leaveTangent=-2.2224, time=0.4375, value=0.8228}
           {arriveTangent=-5.133, interpMode=RCIM_Cubic, leaveTangent=-5.133, time=0.5, value=0.1046}
           {arriveTangent=-0.2282, interpMode=RCIM_Cubic, leaveTangent=-0.2282, time=0.5625, value=0.1812}
           {
            arriveTangent=-0.4342
            arriveTangentWeight=0.0279
            interpMode=RCIM_Cubic
            leaveTangent=-0.4342
            leaveTangentWeight=0.0759
            time=0.875
            value=0.0015
           }
           {arriveTangent=-0.4342, interpMode=RCIM_Cubic, leaveTangent=-0.4342, time=0.9375, value=0.0048}
           {arriveTangent=12.092, interpMode=RCIM_Cubic, leaveTangent=12.092, time=0.9688, value=0.0048}
           <at 1 0.7605 RCIM_Cubic>
          ]}
         GreenCurve {keys [
           {interpMode=RCIM_Cubic, leaveTangentWeight=0.0208, time=0, value=0.1441}
           {
            arriveTangent=-0.8979
            arriveTangentWeight=0.0208
            interpMode=RCIM_Cubic
            leaveTangent=-0.8979
            leaveTangentWeight=0.0625
            time=0.0312
            value=0.0319
           }
           {arriveTangent=2.0457, interpMode=RCIM_Cubic, leaveTangent=2.0457, time=0.125, value=0.0319}
           {arriveTangent=2.2592, interpMode=RCIM_Cubic, leaveTangent=2.2592, time=0.1875, value=0.3515}
           {arriveTangent=1.3088, interpMode=RCIM_Cubic, leaveTangent=1.3088, time=0.4375, value=0.7379}
           {arriveTangent=0.365, interpMode=RCIM_Cubic, leaveTangent=0.365, time=0.5, value=0.7605}
           {arriveTangent=-1.555, interpMode=RCIM_Cubic, leaveTangent=-1.555, time=0.5625, value=0.7835}
           {
            arriveTangent=-1.7312
            arriveTangentWeight=0.0671
            interpMode=RCIM_Cubic
            leaveTangent=-1.7312
            leaveTangentWeight=0.1825
            time=0.875
            value=0.3663
           }
           {arriveTangent=-1.7312, interpMode=RCIM_Cubic, leaveTangent=-1.7312, time=0.9375, value=0.0802}
           {arriveTangent=1.0225, interpMode=RCIM_Cubic, leaveTangent=1.0225, time=0.9688, value=0.0802}
           <at 1 0.1441 RCIM_Cubic>
          ]}
         RedCurve {keys [
           {interpMode=RCIM_Cubic, leaveTangentWeight=0.0208, time=0, value=0.6867}
           {arriveTangent=-5.132, interpMode=RCIM_Cubic, leaveTangent=-5.132, time=0.0312, value=0.1119}
           {arriveTangent=-0.5198, interpMode=RCIM_Cubic, leaveTangent=-0.5198, time=0.125, value=0.0452}
           {arriveTangent=0.1677, interpMode=RCIM_Cubic, leaveTangent=0.1677, time=0.1875, value=0.0307}
           {arriveTangent=0.4007, interpMode=RCIM_Cubic, leaveTangent=0.4007, time=0.4375, value=0.0976}
           {arriveTangent=5.4259, interpMode=RCIM_Cubic, leaveTangent=5.4259, time=0.5, value=0.1559}
           {arriveTangent=1.5243, interpMode=RCIM_Cubic, leaveTangent=1.5243, time=0.5625, value=0.7758}
           {
            arriveTangent=0.1156
            arriveTangentWeight=0.0234
            interpMode=RCIM_Cubic
            leaveTangent=0.1156
            leaveTangentWeight=0.0635
            time=0.875
            value=0.807
           }
           {arriveTangent=0.1156, interpMode=RCIM_Cubic, leaveTangent=0.1156, time=0.9375, value=0.8228}
           {arriveTangent=-2.1776, interpMode=RCIM_Cubic, leaveTangent=-2.1776, time=0.9688, value=0.8228}
           <at 1 0.6867 RCIM_Cubic>
          ]}
        }}
       CurveIndex=&Frac_Float(Float=&Divide_Float(A=$Particles.Initial.ColorIndexFloat, B=16))
      ))
     "Float (Alpha)"=&FN_FrontEndFloatModifier(Gameplay=0.3, Front End=0.15)
    )) {isSetParameters=true, module=SetVariables_#7}
  ]
  particleUpdateScript [
   &ParticleState()
   &ScaleMeshSize(Scale Factor=&VectorFromFloat(Value=&FloatFromCurve(FloatCurve=DataInterface() {propertyValues {Curve {
         defaultValue=1
         keys [
          {arriveTangent=9.16, interpMode=RCIM_Cubic, leaveTangent=9.16, tangentMode=RCTM_User, time=0, value=0}
          {arriveTangent=3.1761, interpMode=RCIM_Cubic, leaveTangent=3.1761, time=0.1378, value=0.7088}
          {
           arriveTangent=0.0339
           arriveTangentWeight=0.5121
           interpMode=RCIM_Cubic
           leaveTangent=0.0339
           leaveTangentWeight=0.5121
           tangentMode=RCTM_User
           time=0.3149
           value=1
          }
          {arriveTangent=-1.9896, interpMode=RCIM_Cubic, leaveTangent=-1.9896, tangentMode=RCTM_User, time=0.8668, value=0.65}
          {arriveTangent=-5.4864, interpMode=RCIM_Cubic, leaveTangent=-5.4864, tangentMode=RCTM_User, time=1, value=0}
         ]
        }}})))
   &SkeletalMeshLocation(
    MaskSpawnGroup=false
    Mesh Sampling Type=::NewEnumerator5
    Orientation Sampling=::NewEnumerator1
    Position Sampling=::NewEnumerator2
    Skeletal Mesh=$System.Skeletal Mesh Reader
    Socket Sampling Mode=::NewEnumerator5
   )
   &ShapeLocation(
    MaskSpawnGroup=false
    Override Seed=false
    Ring / Disc Distribution Mode:ENiagaraSphereDistributionMode=::NewEnumerator1
    Ring Radius=&Lerp_Float(A=20, B=36, Alpha=$Particles.Initial.RadiusRandom)
    Shape Origin=&Add_Vector(A=Linked() {linkedVariable {name=Particles.SkeletalMeshLocation.SampledPosition, type=NiagaraPosition}}, B=$Particles.Initial.HeightRandom)
    Shape Primitive=::NewEnumerator4
    U Position=&Add_Float(
     A=&Add_Float(
      A=$Particles.Initial.ParticleCirclePosition
      B=&Multiply_FloatByInt(
       Float=&FloatFromCurve(
        FloatCurve=DataInterface() {propertyValues {Curve {
           defaultValue=1
           keys [
            {arriveTangent=2.4974, interpMode=RCIM_Cubic, leaveTangent=2.4974, tangentMode=RCTM_User, time=0, value=0.3294}
            {arriveTangent=0.5706, leaveTangent=0.5706, time=0.3468, value=1}
            {arriveTangent=-0.5229, interpMode=RCIM_Cubic, leaveTangent=-0.5229, time=0.8, value=1}
            <at 1 0.6584 RCIM_Cubic>
           ]
          }}}
        Scale Curve=-0.05
       )
       Integer=$Particles.Initial.Order
      )
     )
     B=&Multiply_Float(A=$Engine.System.Age, B=&Lerp_Float(A=0.42, B=0.1, Alpha=$Particles.Initial.RadiusRandom))
    )
    UseOwnerScale=true
   )
   &UpdateMeshOrientation(
    Facing Direction=&NormalizeVector(
     Vector To Normalize=&SubtractVectorFromPosition(
      Position=$Particles.Position
      Vector=&Add_Vector(A=Linked() {linkedVariable {name=Particles.SkeletalMeshLocation.SampledPosition, type=NiagaraPosition}}, B=$Particles.Initial.HeightRandom)
     )
     Safe Normalize=true
     Safe Fallback Threshold=0
     Safe Fallback Vector=<0 0 0.001>
    )
    Facing Interpolation=false
    Orientation Method=::NewEnumerator1
    Override Quat=false
    TransformFacing=false
    TransformUp=false
   )
   &SolveForcesAndVelocity(Clamp Velocity=false, Limit Acceleration=false)
   /SetVariables(Particles.VoxelParams:Vector4f=&MakeVector4(X=&RandomRangeFloat(Minimum=0.4, Maximum=0.8), Y=0, Z=0, W=0)) {isSetParameters=true, module=SetVariables_#8}
   /SetVariables(
    Particles.Glow_Base_CameraOffset=-5
    Particles.Glow_Base_Size:Vector2f=&Vector2DFromFloat(Value=&Multiply_Float(A=&MakeFloatFromVector(Vector=$Particles.Scale, Channel=::NewEnumerator0), B=1230))
   ) {isSetParameters=true, module=SetVariables_#9}
   &ScaleColor(
    Curve Index=&Add_Float(A=$Particles.NormalizedAge, B=&RandomRangeFloat(Minimum=-0.1, Maximum=0.1))
    Linear Color Curve=DataInterface() {propertyValues {
      AlphaCurve {keys [<at 0 1 RCIM_Constant>]}
      BlueCurve {keys [<at 0.0149 1>,<at 0.0276 5 RCIM_Constant>,<at 0.0333 0.3 RCIM_Constant>,<at 0.0418 1.5 RCIM_Constant>,<at 0.056 0.7 RCIM_Constant>,<at 0.0773 1 RCIM_Constant>,<at 0.4544 1 RCIM_Constant>,<at 0.4572 1.4 RCIM_Constant>,<at 0.47 1 RCIM_Constant>,<at 0.5962 1 RCIM_Constant>,<at 0.6018 0.8 RCIM_Constant>,<at 0.6118 1 RCIM_Constant>,<at 0.8542 1 RCIM_Constant>,<at 0.867 0.6 RCIM_Constant>,<at 0.8755 1.5 RCIM_Constant>,<at 0.8854 1 RCIM_Constant>,<at 0.8953 0.3 RCIM_Constant>,<at 0.8995 1 RCIM_Constant>]}
      GreenCurve {keys [<at 0.0149 1>,<at 0.0276 5 RCIM_Constant>,<at 0.0333 0.3 RCIM_Constant>,<at 0.0418 1.5 RCIM_Constant>,<at 0.056 0.7 RCIM_Constant>,<at 0.0773 1 RCIM_Constant>,<at 0.4544 1 RCIM_Constant>,<at 0.4572 1.4 RCIM_Constant>,<at 0.47 1 RCIM_Constant>,<at 0.5962 1 RCIM_Constant>,<at 0.6018 0.8 RCIM_Constant>,<at 0.6118 1 RCIM_Constant>,<at 0.8542 1 RCIM_Constant>,<at 0.867 0.6 RCIM_Constant>,<at 0.8755 1.5 RCIM_Constant>,<at 0.8854 1 RCIM_Constant>,<at 0.8953 0.3 RCIM_Constant>,<at 0.8995 1 RCIM_Constant>]}
      RedCurve {keys [<at 0.0149 1>,<at 0.0276 5 RCIM_Constant>,<at 0.0333 0.3 RCIM_Constant>,<at 0.0418 1.5 RCIM_Constant>,<at 0.056 0.7 RCIM_Constant>,<at 0.0773 1 RCIM_Constant>,<at 0.4544 1 RCIM_Constant>,<at 0.4572 1.4 RCIM_Constant>,<at 0.47 1 RCIM_Constant>,<at 0.5962 1 RCIM_Constant>,<at 0.6018 0.8 RCIM_Constant>,<at 0.6118 1 RCIM_Constant>,<at 0.8542 1 RCIM_Constant>,<at 0.867 0.6 RCIM_Constant>,<at 0.8755 1.5 RCIM_Constant>,<at 0.8854 1 RCIM_Constant>,<at 0.8953 0.3 RCIM_Constant>,<at 0.8995 1 RCIM_Constant>]}
     }}
    Scale Mode=2
   )
   /SetVariables(
    Particles.JitterDirection:Vector3f=&RandomRangeVector(Minimum=<-2 -2 -2>, Maximum=<2 2 2>, Evaluation Type=::NewEnumerator1)
    Particles.JitterFallof:NiagaraFloat=&FloatFromCurve(FloatCurve=DataInterface() {propertyValues {Curve {
        defaultValue=0
        keys [
         {interpMode=RCIM_Cubic, tangentMode=RCTM_User, time=0, value=0}
         {interpMode=RCIM_Cubic, tangentMode=RCTM_User, time=0.5, value=1}
         {tangentMode=RCTM_User, time=1, value=0}
        ]
       }}})
    Particles.JitterStrength:NiagaraFloat=&FloatFromCurve(
     FloatCurve=DataInterface() {propertyValues {Curve {keys [<at 0 0 RCIM_Constant>,<at 0.8991 1 RCIM_Constant>]}}}
     CurveIndex=&RandomRangeFloat(Minimum=0, Maximum=1, Evaluation Type=::NewEnumerator1)
    )
   ) {isSetParameters=true, module=SetVariables_#10}
   /SetVariables(Particles.JitterPosition:NiagaraPosition=&Add_Vector(
     A=Linked() {linkedVariable {name=Particles.Position, type=NiagaraPosition}}
     B=&Multiply_VectorByFloat(
      Vector=$Particles.JitterDirection
      Float=&Multiply_Float(A=$Particles.JitterStrength, B=$Particles.JitterFallof)
     )
    )) {isSetParameters=true, module=SetVariables_#11}
   /SetVariables(
    Particles.GlitchColor:LinearColor=&MakeLinearColorFromVectorAndFloat(
     "Vector (RGB)"=&MakeVectorFromLinearColorRGB(Color=&ColorFromCurve(
       ColorCurve=DataInterface() {propertyValues {
         AlphaCurve {keys [<at 0 1 RCIM_Cubic>]}
         BlueCurve {keys [
           <at 0 0.7523>
           <at 0.1362 1>
           <at 0.3005 1>
           <at 0.4906 1>
           {arriveTangent=-2.3178, interpMode=RCIM_Cubic, leaveTangent=-2.3178, time=0.6656, value=1}
           <at 0.769 0.3547>
           <at 0.9227 0.1565>
           <at 1 0.1478 RCIM_Cubic>
          ]}
         GreenCurve {keys [
           <at 0 0.1632>
           <at 0.1362 0.0609>
           <at 0.3005 0.087>
           <at 0.4906 0.2463>
           {arriveTangent=2.7072, interpMode=RCIM_Cubic, leaveTangent=2.7072, time=0.6656, value=0.7595}
           <at 0.769 1>
           <at 0.9227 0.9528>
           <at 1 0.2044 RCIM_Cubic>
          ]}
         RedCurve {keys [
           <at 0 1>
           <at 0.1362 0.4789>
           <at 0.3005 0.3414>
           <at 0.4906 0.1019>
           {arriveTangent=0.4459, interpMode=RCIM_Cubic, leaveTangent=0.4459, time=0.6656, value=0}
           <at 0.769 0.2261>
           <at 0.9227 1>
           <at 1 1 RCIM_Cubic>
          ]}
        }}
       CurveIndex=&Frac_Float(Float=&Add_Float(
         A=&RandomRangeFloat(Minimum=0.7, Maximum=1.7)
         B=&RandomRangeFloat(Minimum=-0.7, Maximum=0.7, Evaluation Type=::NewEnumerator1)
        ))
      ))
     "Float (Alpha)"=0.8
    )
    Particles.GlitchPosition:NiagaraPosition=&Add_Vector(
     A=Linked() {linkedVariable {name=Particles.Position, type=NiagaraPosition}}
     B=&Multiply_VectorByFloat(
      Vector=$Particles.JitterDirection
      Float=&Multiply_Float(A=&Multiply_Float(A=$Particles.JitterStrength, B=$Particles.JitterFallof), B=-1.5)
     )
    )
    Particles.GlitchScale:Vector3f=&Multiply_VectorByFloat(Vector=$Particles.Scale, Float=$Particles.JitterStrength)
   ) {isSetParameters=true, module=SetVariables_#12}
  ]
  rendererClasses [&NiagaraMeshRendererProperties, &NiagaraSpriteRendererProperties]
  renderers [
   {rendererClass=&NiagaraMeshRendererProperties, rendererIndex=0}
   {rendererClass=&NiagaraSpriteRendererProperties, rendererIndex=1}
   {rendererClass=&NiagaraMeshRendererProperties, rendererIndex=2}
  ]
 }
 Glitch_Blocks {
  emitterUpdateScript [
   &EmitterState(Life Cycle Mode=::NewEnumerator1, Loop Behavior=::NewEnumerator1, Loop Delay=1, Loop Duration Mode=::NewEnumerator1, UseLoopDelay=true)
   &SpawnRate(Spawn Probability=$System.HeroicSpawn, SpawnRate=15, Use Spawn Probability=true)
   &SpawnRate(Spawn Probability=$System.HeroicSpawn, SpawnGroup=1, SpawnRate=9, Use Spawn Probability=true) {module=SpawnRate001}
   &SpawnRate(Spawn Probability=$System.HeroicSpawn, SpawnGroup=2, SpawnRate=9, Use Spawn Probability=true) {module=SpawnRate002}
  ]
  particleSpawnScript [
   /SetVariables(Particles.VisibilityTag=&MaskIntBySpawnGroup(
     Masked Int=0
     Passthrough Int=&MakeCustomIntFromBool(
      Bool=&SetBoolByFloatComparison(A=&RandomRangeFloat(Minimum=0, Maximum=1), B=0.25)
      True INT=1
      False INT=0
     )
     Spawn Group Mask=2
    )) {isSetParameters=true, module=SetVariables_#13}
   &InitializeParticle(
    Color=&MaskLinearColor(
     Mask Source=1
     Mask Parameter=$Particles.VisibilityTag
     Mask=1
     Masked Linear Color=<rgba 3 3 3 6>
     Passthrough Linear Color=<rgba 1 1 1 0.8>
    )
    Lifetime=&MaskFloatBySpawnGroup(Masked Float=0.3, Passthrough Float=&RandomRangeFloat(Minimum=0.15, Maximum=0.19), Spawn Group Mask=2)
    Mesh Scale=&MaskVectorBySpawnGroup(
     Masked Vector=<0.275 0.275 0.275>
     Passthrough Vector=&VectorFromCurve(
      VectorCurve=DataInterface() {propertyValues {
        XCurve {keys [<at 0 0.373>,<at 0.1624 0.1237 RCIM_Constant>,<at 0.1957 0.2565>,<at 0.5247 0.1645 RCIM_Constant>,<at 0.5602 1.258>,<at 1 0.8335 RCIM_Constant>]}
        YCurve {keys [<at 0 0.02 RCIM_Constant>,<at 0.1957 0.8136>,<at 0.5247 0.5854 RCIM_Constant>,<at 0.5602 0.4973>,<at 1 0.4039 RCIM_Constant>]}
        ZCurve {keys [<at 0 0.8813>,<at 0.1624 0.361 RCIM_Constant>,<at 0.1957 0.05 RCIM_Constant>,<at 0.5602 0.05 RCIM_Constant>]}
       }}
      CurveIndex=&RandomRangeFloat(Minimum=0, Maximum=1)
     )
     Spawn Group Mask=2
    )
    Mesh Scale Mode=::NewEnumerator4
    UsePositionOffset=false
    Write Material Random=true
   )
   &CameraOffset()
   /SetVariables(Particles.HeightRandom:NiagaraFloat=&RandomRangeFloat(Minimum=-1, Maximum=1)) {isSetParameters=true, module=SetVariables_#14}
   &ShapeLocation(
    MaskSpawnGroup=false
    Offset=&Add_Vector(
     A=&SelectVectorFromArray(
      Direct Array Index=&SpawnGroup()
      Vector Selection Array=DataInterface() {propertyValues {FloatData [<0 0 0>, <0 0 30>, <0 0 5>]}}
     )
     B=&Multiply_VectorByFloat(
      Vector=&SelectVectorFromArray(
       Direct Array Index=&SpawnGroup()
       Vector Selection Array=DataInterface() {propertyValues {FloatData [<0 0 0>, <0 0 10>, <0 0 10>]}}
      )
      Float=$Particles.HeightRandom
     )
    )
    Offset Coordinate Space=::Local
    Offset Mode=::NewEnumerator0
    Override Seed=false
    Ring Radius=&Add_Float(
     A=&SelectFloatFromArray(Direct Array Index=&SpawnGroup(), Float Selection Array=DataInterface() {propertyValues {FloatData [73,17,60]}})
     B=&Multiply_Float(
      A=&SelectFloatFromArray(Direct Array Index=&SpawnGroup(), Float Selection Array=DataInterface() {propertyValues {FloatData [8,2,8]}})
      B=&RandomRangeFloat(Minimum=-1, Maximum=1)
     )
    )
    Shape Primitive=::NewEnumerator4
    UseOwnerScale=true
   )
   &InitialMeshOrientation(
    Mesh Axis To Orient=<0 0 1>
    Mesh Orientation Mode=::NewEnumerator2
    Orientation Coordinate Space=::Local
    Orientation Vector=<0 0 1>
    Rotation=&SelectVectorFromArray(
     Direct Array Index=&MaskIntBySpawnGroup(Masked Int=&RandomRangeInt(Minimum=0, Maximum=3), Passthrough Int=0, Spawn Group Mask=2)
     Vector Selection Array=DataInterface() {propertyValues {FloatData [<0 0 0>, <0.25 0 0>, <0 0.25 0>, <0 0 0.25>]}}
    )
    Use Rotation Vector=true
   )
   &AddVelocity(Velocity=&MaskVectorBySpawnGroup(
     Masked Vector=&RandomRangeVector(Minimum=<-12 -12 -6>, Maximum=<12 12 6>)
     Passthrough Vector=&RandomRangeVector(Minimum=<-14 -9 0>, Maximum=<14 9 0>)
     Spawn Group Mask=2
    ))
  ]
  particleUpdateScript [
   &ParticleState()
   &Drag(Drag=3, Rotational Drag=3, Use Linear Drag=true, Use Rotational Drag=true)
   &SolveForcesAndVelocity(Clamp Velocity=false, Limit Acceleration=false)
   &ScaleColor(
    Scale Alpha=&FloatFromCurve(FloatCurve=DataInterface() {propertyValues {Curve {
        defaultValue=1
        keys [
         {arriveTangent=1.2192, interpMode=RCIM_Cubic, leaveTangent=1.2192, tangentMode=RCTM_User, time=0, value=0}
         {arriveTangent=3.0011, interpMode=RCIM_Cubic, leaveTangent=3.0011, time=0.1697, value=0.4382}
         {interpMode=RCIM_Cubic, tangentMode=RCTM_User, time=0.3332, value=1}
         {arriveTangent=-1.1786, tangentMode=RCTM_User, time=1, value=0}
        ]
       }}})
    Scale RGB=&MakeVectorFromLinearColorRGB(Color=&ColorFromCurve(
      ColorCurve=DataInterface() {propertyValues {
        AlphaCurve {keys [<at 0 1 RCIM_Cubic>]}
        BlueCurve {keys [
          <at 0 0.7523>
          <at 0.0542 1>
          <at 0.1804 1>
          <at 0.3304 1>
          {arriveTangent=-1.5974, interpMode=RCIM_Cubic, leaveTangent=-1.5974, time=0.5413, value=1}
          <at 0.7344 0.3547>
          <at 0.7983 0.1565>
          <at 1 0.1478 RCIM_Cubic>
         ]}
        GreenCurve {keys [
          <at 0 0.1632>
          <at 0.0542 0.0609>
          <at 0.1804 0.087>
          <at 0.3304 0.2463>
          {arriveTangent=1.8657, interpMode=RCIM_Cubic, leaveTangent=1.8657, time=0.5413, value=0.7595}
          <at 0.7344 1>
          <at 0.7983 0.9528>
          <at 1 0.2044 RCIM_Cubic>
         ]}
        RedCurve {keys [
          <at 0 1>
          <at 0.0542 0.4789>
          <at 0.1804 0.3414>
          <at 0.3304 0.1019>
          {arriveTangent=0.3073, interpMode=RCIM_Cubic, leaveTangent=0.3073, time=0.5413, value=0}
          <at 0.7344 0.2261>
          <at 0.7983 1>
          <at 1 1 RCIM_Cubic>
         ]}
       }}
      CurveIndex=&Frac_Float(Float=&Add_Float(
        A=&RandomRangeFloat(Minimum=0.7, Maximum=1.7)
        B=&Multiply_Float(
         A=&RandomRangeFloat(Minimum=-0.7, Maximum=0.7, Evaluation Type=::NewEnumerator1)
         B=$Particles.NormalizedAge
        )
       ))
     ))
    ScaleA=true
    ScaleRGB=true
   )
   &ScaleMeshSize(Scale Factor=&MaskVectorBySpawnGroup(
     Masked Vector=<1 1 1>
     Passthrough Vector=&MakeVector(
      X=&FloatFromCurve(FloatCurve=DataInterface() {propertyValues {Curve {
          defaultValue=1
          keys [
           {arriveTangent=1.4643, time=0, value=0.7528}
           {arriveTangent=0.0248, interpMode=RCIM_Cubic, leaveTangent=0.0248, tangentMode=RCTM_User, time=0.2571, value=1}
           {arriveTangent=-0.187, leaveTangent=-0.187, time=0.7725, value=0.9775}
           {arriveTangent=-0.5952, tangentMode=RCTM_User, time=1, value=0.8611}
          ]
         }}})
      Y=1
      Z=1
     )
     Spawn Group Mask=2
    ))
   /SetVariables(Particles.JitterPosition:NiagaraPosition=&Add_Vector(
     A=Linked() {linkedVariable {name=Particles.Position, type=NiagaraPosition}}
     B=&Multiply_VectorByFloat(
      Vector=&RandomRangeVector(Minimum=<-1.5 -1.5 -1.5>, Maximum=<1.5 1.5 1.5>, Evaluation Type=::NewEnumerator1)
      Float=&FloatFromCurve(
       FloatCurve=DataInterface() {propertyValues {Curve {keys [<at 0 0 RCIM_Constant>,<at 0.8991 1 RCIM_Constant>]}}}
       CurveIndex=&RandomRangeFloat(Minimum=0, Maximum=1, Evaluation Type=::NewEnumerator1)
       Scale Curve=&FloatFromCurve(FloatCurve=DataInterface() {propertyValues {Curve {
           defaultValue=0
           keys [
            {interpMode=RCIM_Cubic, tangentMode=RCTM_User, time=0, value=0}
            {interpMode=RCIM_Cubic, tangentMode=RCTM_User, time=0.3739, value=1}
            {tangentMode=RCTM_User, time=1, value=0}
           ]
          }}})
      )
     )
    )) {isSetParameters=true}
  ]
  rendererClasses [&NiagaraMeshRendererProperties]
  renderers [
   {rendererClass=&NiagaraMeshRendererProperties, rendererIndex=0}
   {rendererClass=&NiagaraMeshRendererProperties, rendererIndex=1}
  ]
 }
 Glitch_Major {
  emitterUpdateScript [
   &EmitterState(Life Cycle Mode=::NewEnumerator1, Loop Behavior=::NewEnumerator1, Loop Delay=1, Loop Duration Mode=::NewEnumerator1, UseLoopDelay=true)
   &SpawnBurst_Instantaneous(Spawn Probability=$System.HeroicSpawn, Use Spawn Probability=true, UseLoopCountLimit=false)
   &SpawnBurst_Instantaneous(Spawn Group=1, Spawn Probability=$System.HeroicSpawn, Use Spawn Probability=true, UseLoopCountLimit=false) {module=SpawnBurst_Instantaneous001}
  ]
  particleSpawnScript [
   &InitializeParticle(
    Color=<rgba 7 7 7 5>
    Position=&MaskVectorBySpawnGroup(Masked Vector=<0 0 0>, Passthrough Vector=<0 0 40>)
    Position Mode=::NewEnumerator1
    Sprite Size=/Niagara/DynamicInputs/Helpers/MaskVector2d(
     Mask Source:NiagaraInt32=0
     Mask Parameter:NiagaraInt32=?
     Mask:NiagaraInt32=0
     Masked Vector2D:Vector2f=<60 120>
     Passthrough Vector2D:Vector2f=<40 40>
    )
    Sprite Size Mode=::NewEnumerator4
    UsePositionOffset=false
    Write Material Random=true
   )
   &CameraOffset()
  ]
  particleUpdateScript [
   &ParticleState(Kill Particles When Lifetime Has Elapsed=false, Loop Particles Lifetime=true)
   &DynamicMaterialParameters(
    Index 0 Param 1=&FloatFromCurve(
     FloatCurve=DataInterface() {propertyValues {Curve {keys [<at 0 0 RCIM_Constant>,<at 0.2974 0.3 RCIM_Constant>,<at 0.94 0.6 RCIM_Constant>,<at 1 0 RCIM_Constant>]}}}
     CurveIndex=&RandomRangeFloat(Minimum=0, Maximum=1, Evaluation Type=::NewEnumerator1)
    )
    Module.Index0Float0Write=true
    Module.Index0Float1Write=false
    Module.Index0Float2Write=false
    Module.Index0Float3Write=false
   )
   /SetVariables(Particles.SpriteSize=&MaskVector2dBySpawnGroup(
     Masked Vector2D=&MakeVector2D(
      X=&Lerp_Float(
       A=30
       B=120
       Alpha=&FloatFromCurve(FloatCurve=DataInterface() {propertyValues {Curve {keys [<at 0.0992 0>,<at 1 1>]}}}, CurveIndex=$System.SystemViewDot)
      )
      Y=300
     )
     Passthrough Vector2D=<45 40>
     Spawn Group Mask=0
    )) {isSetParameters=true}
   &ScaleColor(
    Curve Index=&RandomRangeFloat(Minimum=0, Maximum=1, Evaluation Type=::NewEnumerator1)
    Linear Color Curve=DataInterface() {propertyValues {
      AlphaCurve {keys [<at 0 1>]}
      BlueCurve {keys [<at 0.022 0.8272>,<at 0.0418 0.0787>,<at 0.0702 0>,<at 0.2417 1>,<at 0.5848 1>,<at 0.8641 1>,<at 0.884 1>,<at 0.8996 1>,<at 1 0.4067>]}
      GreenCurve {keys [<at 0.022 1>,<at 0.0418 1>,<at 0.0702 1>,<at 0.2417 0.8213>,<at 0.5848 0.3937>,<at 0.8641 0.3937>,<at 0.884 0.2054>,<at 0.8996 0.1479>,<at 1 0.24>]}
      RedCurve {keys [<at 0.022 0.762>,<at 0.0418 0.4767>,<at 0.0702 0.0211>,<at 0.2417 0.0231>,<at 0.5848 0>,<at 0.8641 0>,<at 0.884 0.505>,<at 0.8996 0.7618>,<at 1 1>]}
     }}
    Scale Mode=2
   )
  ]
  rendererClasses [&NiagaraSpriteRendererProperties]
  renderers [{rendererClass=&NiagaraSpriteRendererProperties}]
 }
 Glitch_Minor {
  emitterUpdateScript [
   &EmitterState(Life Cycle Mode=::NewEnumerator1, Loop Behavior=::NewEnumerator1, Loop Delay=1, Loop Duration Mode=::NewEnumerator1, UseLoopDelay=true)
   &SpawnRate(Spawn Probability=$System.HeroicSpawn, SpawnRate=7, Use Spawn Probability=true)
   &SpawnRate(Spawn Probability=$System.HeroicSpawn, SpawnGroup=1, SpawnRate=4, Use Spawn Probability=true) {module=SpawnRate001}
  ]
  particleSpawnScript [
   &InitializeParticle(
    Color=<rgba 6 6 6 5>
    Lifetime=0.4
    Mesh Scale Max=<1 1 0.6>
    Mesh Scale Min=<1 0.5 0.2>
    Mesh Scale Mode=::NewEnumerator2
    Sprite Size Max=<6 40>
    Sprite Size Min=<3 40>
    Sprite Size Mode=::NewEnumerator2
    UsePositionOffset=false
    Write Material Random=true
   )
   &CameraOffset()
   /SetVariables(Particles.HeightRandom:NiagaraFloat=&RandomRangeFloat(Minimum=0, Maximum=1)) {isSetParameters=true, module=SetVariables_#14}
   &ShapeLocation(
    MaskSpawnGroup=false
    Offset=&Add_Vector(
     A=&Lerp_Vector(A=<0 0 -10>, B=<0 0 10>, Alpha=$Particles.HeightRandom)
     B=&MaskVectorBySpawnGroup(Masked Vector=<0 0 0>, Passthrough Vector=<0 0 30>)
    )
    Offset Coordinate Space=::Local
    Offset Mode=::NewEnumerator0
    Override Seed=false
    Ring Radius=&MaskFloatBySpawnGroup(Masked Float=&Lerp_Float(A=80, B=30, Alpha=$Particles.HeightRandom), Passthrough Float=20)
    Shape Primitive=::NewEnumerator4
    UseOwnerScale=true
   )
   /SetVariables(Particles.VisibilityTag=&RandomRangeInt(Minimum=0, Maximum=2)) {isSetParameters=true, module=SetVariables_#13}
   &InitialMeshOrientation(
    Mesh Axis To Orient=<0 0 1>
    Mesh Orientation Mode=::NewEnumerator2
    Orientation Coordinate Space=::Local
    Orientation Vector=<0 0 1>
    Use Rotation Vector=false
   )
  ]
  particleUpdateScript [
   &ParticleState()
   &ScaleColor(
    Scale Alpha=&FloatFromCurve(FloatCurve=DataInterface() {propertyValues {Curve {keys [<at 0.4547 1>,<at 1 0>]}}})
    Scale RGB=&MakeVectorFromLinearColorRGB(Color=&ColorFromCurve(
      ColorCurve=DataInterface() {propertyValues {
        AlphaCurve {keys [<at 0 1>]}
        BlueCurve {keys [<at 0.022 0.8272>,<at 0.0418 0.0787>,<at 0.1118 0>,<at 0.306 1>,<at 0.4505 1>,<at 0.6582 1>,<at 0.7101 1>,<at 0.823 1>,<at 1 0.4067>]}
        GreenCurve {keys [<at 0.022 1>,<at 0.0418 1>,<at 0.1118 1>,<at 0.306 0.8213>,<at 0.4505 0.3937>,<at 0.6582 0.3937>,<at 0.7101 0.2054>,<at 0.823 0.1479>,<at 1 0.24>]}
        RedCurve {keys [<at 0.022 0.762>,<at 0.0418 0.4767>,<at 0.1118 0.0211>,<at 0.306 0.0231>,<at 0.4505 0>,<at 0.6582 0>,<at 0.7101 0.505>,<at 0.823 0.7618>,<at 1 1>]}
       }}
      CurveIndex=&RandomRangeFloat(Minimum=0, Maximum=1, Evaluation Type=::NewEnumerator1)
     ))
    ScaleA=true
    ScaleRGB=true
   )
   &DynamicMaterialParameters(
    Index 0 Param 1=&FloatFromCurve(FloatCurve=DataInterface() {propertyValues {Curve {keys [<at 0 0>,<at 0.0628 1>,<at 0.2274 0.5>,<at 1 0>]}}}, Scale Curve=3)
    Module.Index0Float0Write=true
    Module.Index0Float1Write=false
    Module.Index0Float2Write=false
    Module.Index0Float3Write=false
   )
  ]
  rendererClasses [&NiagaraSpriteRendererProperties]
  renderers [
   {rendererClass=&NiagaraSpriteRendererProperties}
   {rendererClass=&NiagaraSpriteRendererProperties, rendererIndex=1}
   {rendererClass=&NiagaraSpriteRendererProperties, rendererIndex=2}
  ]
 }
 Glow {
  emitterUpdateScript [
   &EmitterState(Life Cycle Mode=::NewEnumerator1, Loop Behavior=::NewEnumerator1, Loop Delay=1, Loop Duration=10, UseLoopDelay=true)
   &SpawnBurst_Instantaneous(Use Spawn Probability=false, UseLoopCountLimit=true)
   &SpawnBurst_Instantaneous(Spawn Group=1, Spawn Probability=$System.HeroicSpawn, Use Spawn Probability=true, UseLoopCountLimit=true) {module=SpawnBurst_Instantaneous001}
   &SpawnBurst_Instantaneous(Loop Count Limit=2, Spawn Group=2, Spawn Probability=$System.HeroicSpawn, Use Spawn Probability=true, UseLoopCountLimit=false) {module=SpawnBurst_Instantaneous002}
   &SpawnBurst_Instantaneous(Spawn Count=3, Spawn Group=3, Spawn Probability=$System.HeroicSpawn, Use Spawn Probability=true, UseLoopCountLimit=true) {module=SpawnBurst_Instantaneous003}
  ]
  particleSpawnScript [
   /SetVariables(Particles.VisibilityTag=&SpawnGroup()) {isSetParameters=true, module=SetVariables_#15}
   &InitializeParticle(
    Lifetime=2
    Mesh Scale=<1 1 1.2476>
    Mesh Scale Mode=::NewEnumerator4
    Sprite Size Mode=::NewEnumerator3
    Uniform Sprite Size=&SelectFloatFromArray(
     Direct Array Index=&SpawnGroup()
     Float Selection Array=DataInterface() {propertyValues {FloatData [98,140,0,45]}}
    )
    UsePositionOffset=false
    Write Material Random=true
   )
  ]
  particleUpdateScript [
   &ParticleState(Kill Particles When Lifetime Has Elapsed=false)
   &SkeletalMeshLocation(
    Bone / Socket Index=&SelectIntFromArray(Direct Array Index=&SpawnGroup(), Int Selection Array=DataInterface() {propertyValues {IntData [1,0,1,3]}})
    MaskSpawnGroup=false
    Mesh Sampling Type=::NewEnumerator5
    Orientation Sampling=::NewEnumerator1
    Skeletal Mesh=$System.Skeletal Mesh Reader
    Socket Sampling Mode=::NewEnumerator5
   )
   /SetVariables(Particles.Brightness:NiagaraFloat=&FloatFromCurve(
     FloatCurve=DataInterface() {propertyValues {Curve {keys [{arriveTangent=2.3838, interpMode=RCIM_Cubic, leaveTangent=2.3838, tangentMode=RCTM_User, time=0, value=0}, <at 0.3655 1 RCIM_Cubic>]}}}
     CurveIndex=$Particles.Age
    )) {isSetParameters=true, module=SetVariables_#16}
   /SetVariables(
    Particles.ColorCrown:LinearColor=&Multiply_LinearColorByFloat(Linear Color=<rgba 1 0.5652 0.1304 0.3913>, Float=$Particles.Brightness)
    Particles.ColorGlow:LinearColor=&Multiply_LinearColorByFloat(Linear Color=<rgba 1 0.6042 0.1396 0.3391>, Float=$Particles.Brightness)
    Particles.ColorHandle:LinearColor=&Multiply_LinearColorByFloat(Linear Color=<rgba 1 0.6486 0.1565 0.15>, Float=$Particles.Brightness)
    Particles.ColorRays:LinearColor=&Multiply_LinearColorByFloat(Linear Color=<rgba 1 1 1 0.2609>, Float=$Particles.Brightness)
   ) {isSetParameters=true, module=SetVariables_#17}
   &Light_Attributes(
    Color=1
    Color Mode=0
    Falloff Exponent=1
    Light Color:LinearColor=<rgba 1 0.5416 0.0104 1>
    Light Exponent=2
    Light Position Offset:Vector3f=<0 -8 0>
    Light Radius=20
    Position=1
    Radius=1
    UsePositionOffset=false
   )
  ]
  rendererClasses [
   &NiagaraSpriteRendererProperties
   &NiagaraLightRendererProperties
   &NiagaraMeshRendererProperties
  ]
  renderers [
   {rendererClass=&NiagaraSpriteRendererProperties}
   {rendererClass=&NiagaraLightRendererProperties, rendererIndex=1}
   {rendererClass=&NiagaraMeshRendererProperties, rendererIndex=2}
   {rendererClass=&NiagaraSpriteRendererProperties, rendererIndex=3}
   {rendererClass=&NiagaraSpriteRendererProperties, rendererIndex=4}
  ]
 }
 Glow001 {
  bEnabled=false
  emitterUpdateScript [
   &EmitterState(Life Cycle Mode=::NewEnumerator1, Loop Behavior=::NewEnumerator1, Loop Delay=1, Loop Duration=10, UseLoopDelay=true)
   &SpawnBurst_Instantaneous(Use Spawn Probability=false, UseLoopCountLimit=true)
   &SpawnBurst_Instantaneous(Spawn Group=1, Spawn Probability=$System.HeroicSpawn, Use Spawn Probability=true, UseLoopCountLimit=true) {module=SpawnBurst_Instantaneous001}
   &SpawnBurst_Instantaneous(Loop Count Limit=2, Spawn Group=2, Spawn Probability=$System.HeroicSpawn, Use Spawn Probability=true, UseLoopCountLimit=false) {module=SpawnBurst_Instantaneous002}
  ]
  particleSpawnScript [
   /SetVariables(Particles.VisibilityTag=&SpawnGroup()) {isSetParameters=true, module=SetVariables_#15}
   &InitializeParticle(
    Lifetime=2
    Mesh Scale=<1 1 1.2476>
    Mesh Scale Mode=::NewEnumerator4
    Sprite Size Mode=::NewEnumerator3
    Uniform Sprite Size=97.518
    UsePositionOffset=false
    Write Material Random=true
   )
  ]
  particleUpdateScript [
   &ParticleState(Kill Particles When Lifetime Has Elapsed=false)
   &SkeletalMeshLocation(
    Bone / Socket Index=&MaskIntBySpawnGroup(Masked Int=0, Passthrough Int=1, Spawn Group Mask=1)
    MaskSpawnGroup=false
    Mesh Sampling Type=::NewEnumerator5
    Orientation Sampling=::NewEnumerator1
    Skeletal Mesh=$System.Skeletal Mesh Reader
    Socket Sampling Mode=::NewEnumerator5
   )
   /SetVariables(Particles.Brightness:NiagaraFloat=&FloatFromCurve(
     FloatCurve=DataInterface() {propertyValues {Curve {keys [{arriveTangent=2.3838, interpMode=RCIM_Cubic, leaveTangent=2.3838, tangentMode=RCTM_User, time=0, value=0}, <at 0.3655 1 RCIM_Cubic>]}}}
     CurveIndex=$Particles.Age
    )) {isSetParameters=true, module=SetVariables_#16}
   /SetVariables(
    Particles.ColorCrown:LinearColor=&Multiply_LinearColorByFloat(Linear Color=<rgba 1 0.5137 0.1219 0.3391>, Float=$Particles.Brightness)
    Particles.ColorGlow:LinearColor=&Multiply_LinearColorByFloat(Linear Color=<rgba 1 0.6042 0.1396 0.3391>, Float=$Particles.Brightness)
    Particles.ColorHandle:LinearColor=&Multiply_LinearColorByFloat(Linear Color=<rgba 1 0.5137 0.1219 0.3391>, Float=$Particles.Brightness)
    Particles.ColorRays:LinearColor=&Multiply_LinearColorByFloat(Linear Color=<rgba 1 1 1 0.2609>, Float=$Particles.Brightness)
   ) {isSetParameters=true, module=SetVariables_#17}
   &Light_Attributes(
    Color=1
    Color Mode=0
    Falloff Exponent=1
    Light Color:LinearColor=<rgba 1 0.5416 0.0104 1>
    Light Exponent=2
    Light Position Offset:Vector3f=<0 -8 0>
    Light Radius=20
    Position=1
    Radius=1
    UsePositionOffset=false
   )
  ]
  rendererClasses [
   &NiagaraSpriteRendererProperties
   &NiagaraLightRendererProperties
   &NiagaraMeshRendererProperties
  ]
  renderers [
   {rendererClass=&NiagaraSpriteRendererProperties}
   {rendererClass=&NiagaraLightRendererProperties, rendererIndex=1}
   {rendererClass=&NiagaraMeshRendererProperties, rendererIndex=2}
   {rendererClass=&NiagaraSpriteRendererProperties, rendererIndex=3}
  ]
 }
 TrailCubes {
  emitterUpdateScript [
   &EmitterState(Life Cycle Mode=::NewEnumerator1, Loop Behavior=::NewEnumerator1, Loop Delay=1.8, Loop Duration Mode=::NewEnumerator1, UseLoopDelay=true)
   &SpawnRate(
    Spawn Probability=&Multiply_Float(
     A=&Multiply_Float(A=$System.HeroicSpawn, B=$System.TrailSpawn)
     B=&FN_FrontEndFloatModifier(Gameplay=1, Front End=0)
    )
    SpawnRate=16
    Use Spawn Probability=true
   )
  ]
  particleSpawnScript [
   &InitializeParticle(
    Color Minimum=<rgba 0.8 0.8 0.8 1>
    Color Mode=::NewEnumerator2
    Lifetime=&RandomRangeFloat(Minimum=0.6, Maximum=0.8)
    Mesh Scale Mode=::NewEnumerator3
    Mesh Uniform Scale=0.03
    UsePositionOffset=false
    Write Material Random=true
   )
   /SetVariables(Particles.CubeMesh:NiagaraInt32=&RandomRangeInt(Minimum=0, Maximum=1)) {isSetParameters=true}
   /SetVariables(Particles.ColorIndexFloat:NiagaraFloat=&RandomRangeFloat(Minimum=0, Maximum=16)) {isSetParameters=true}
   /SetVariables(Particles.Glow_Base_Color:LinearColor=&MakeLinearColorFromVectorAndFloat(
     "Vector (RGB)"=&MakeVectorFromLinearColorRGB(Color=&ColorFromCurve(
       ColorCurve=DataInterface() {propertyValues {
         AlphaCurve {keys [<at 0.5 1 RCIM_Cubic>]}
         BlueCurve {keys [
           {interpMode=RCIM_Cubic, leaveTangentWeight=0.0208, time=0, value=0.7605}
           {
            arriveTangent=-0.9834
            arriveTangentWeight=0.0208
            interpMode=RCIM_Cubic
            leaveTangent=-0.9834
            leaveTangentWeight=0.0625
            time=0.0312
            value=0.6376
           }
           {
            arriveTangent=1.0336
            arriveTangentWeight=0.0625
            interpMode=RCIM_Cubic
            leaveTangent=1.0336
            leaveTangentWeight=0.0208
            time=0.125
            value=0.6376
           }
           {arriveTangent=0.5926, interpMode=RCIM_Cubic, leaveTangent=0.5926, time=0.1875, value=0.7991}
           {arriveTangent=-2.2224, interpMode=RCIM_Cubic, leaveTangent=-2.2224, time=0.4375, value=0.8228}
           {arriveTangent=-5.133, interpMode=RCIM_Cubic, leaveTangent=-5.133, time=0.5, value=0.1046}
           {arriveTangent=-0.2282, interpMode=RCIM_Cubic, leaveTangent=-0.2282, time=0.5625, value=0.1812}
           {
            arriveTangent=-0.4342
            arriveTangentWeight=0.0279
            interpMode=RCIM_Cubic
            leaveTangent=-0.4342
            leaveTangentWeight=0.0759
            time=0.875
            value=0.0015
           }
           {arriveTangent=-0.4342, interpMode=RCIM_Cubic, leaveTangent=-0.4342, time=0.9375, value=0.0048}
           {arriveTangent=12.092, interpMode=RCIM_Cubic, leaveTangent=12.092, time=0.9688, value=0.0048}
           <at 1 0.7605 RCIM_Cubic>
          ]}
         GreenCurve {keys [
           {interpMode=RCIM_Cubic, leaveTangentWeight=0.0208, time=0, value=0.1441}
           {
            arriveTangent=-0.8979
            arriveTangentWeight=0.0208
            interpMode=RCIM_Cubic
            leaveTangent=-0.8979
            leaveTangentWeight=0.0625
            time=0.0312
            value=0.0319
           }
           {arriveTangent=2.0457, interpMode=RCIM_Cubic, leaveTangent=2.0457, time=0.125, value=0.0319}
           {arriveTangent=2.2592, interpMode=RCIM_Cubic, leaveTangent=2.2592, time=0.1875, value=0.3515}
           {arriveTangent=1.3088, interpMode=RCIM_Cubic, leaveTangent=1.3088, time=0.4375, value=0.7379}
           {arriveTangent=0.365, interpMode=RCIM_Cubic, leaveTangent=0.365, time=0.5, value=0.7605}
           {arriveTangent=-1.555, interpMode=RCIM_Cubic, leaveTangent=-1.555, time=0.5625, value=0.7835}
           {
            arriveTangent=-1.7312
            arriveTangentWeight=0.0671
            interpMode=RCIM_Cubic
            leaveTangent=-1.7312
            leaveTangentWeight=0.1825
            time=0.875
            value=0.3663
           }
           {arriveTangent=-1.7312, interpMode=RCIM_Cubic, leaveTangent=-1.7312, time=0.9375, value=0.0802}
           {arriveTangent=1.0225, interpMode=RCIM_Cubic, leaveTangent=1.0225, time=0.9688, value=0.0802}
           <at 1 0.1441 RCIM_Cubic>
          ]}
         RedCurve {keys [
           {interpMode=RCIM_Cubic, leaveTangentWeight=0.0208, time=0, value=0.6867}
           {arriveTangent=-5.132, interpMode=RCIM_Cubic, leaveTangent=-5.132, time=0.0312, value=0.1119}
           {arriveTangent=-0.5198, interpMode=RCIM_Cubic, leaveTangent=-0.5198, time=0.125, value=0.0452}
           {arriveTangent=0.1677, interpMode=RCIM_Cubic, leaveTangent=0.1677, time=0.1875, value=0.0307}
           {arriveTangent=0.4007, interpMode=RCIM_Cubic, leaveTangent=0.4007, time=0.4375, value=0.0976}
           {arriveTangent=5.4259, interpMode=RCIM_Cubic, leaveTangent=5.4259, time=0.5, value=0.1559}
           {arriveTangent=1.5243, interpMode=RCIM_Cubic, leaveTangent=1.5243, time=0.5625, value=0.7758}
           {
            arriveTangent=0.1156
            arriveTangentWeight=0.0234
            interpMode=RCIM_Cubic
            leaveTangent=0.1156
            leaveTangentWeight=0.0635
            time=0.875
            value=0.807
           }
           {arriveTangent=0.1156, interpMode=RCIM_Cubic, leaveTangent=0.1156, time=0.9375, value=0.8228}
           {arriveTangent=-2.1776, interpMode=RCIM_Cubic, leaveTangent=-2.1776, time=0.9688, value=0.8228}
           <at 1 0.6867 RCIM_Cubic>
          ]}
        }}
       CurveIndex=&Frac_Float(Float=&Divide_Float(A=$Particles.Initial.ColorIndexFloat, B=16))
      ))
     "Float (Alpha)"=&FN_FrontEndFloatModifier(Gameplay=0.15, Front End=0.08)
    )) {isSetParameters=true, module=SetVariables_#7}
   &ShapeLocation(Box Size=<135 135 4>, MaskSpawnGroup=false, Override Seed=false, Shape Primitive=::NewEnumerator2, Surface Only=false, UseOwnerScale=true)
   &AddVelocity(Velocity=&RandomRangeVector(Minimum=<-1 -1 -1>, Maximum=<1 1 1>), Velocity Speed Scale=5)
   &AddVelocity(Velocity=<0 -180 0>, Velocity Speed Scale=&FN_FrontEndFloatModifier(Gameplay=0, Front End=1)) {module=AddVelocity001}
  ]
  particleUpdateScript [
   &ParticleState()
   &InheritSourceMovement(Applied Position Scale=<0.8 0.8 0.7>, ApplyForce=false, ApplyPosition=true, ApplyPositionFalloff=false)
   &SolveForcesAndVelocity(Clamp Velocity=false, Limit Acceleration=false)
   &SkeletalMeshLocation(
    Bone / Socket Index=4
    MaskSpawnGroup=false
    Mesh Sampling Type=::NewEnumerator5
    Orientation Sampling=::NewEnumerator1
    Position Sampling=::NewEnumerator2
    Skeletal Mesh=$System.Skeletal Mesh Reader
    Socket Sampling Mode=::NewEnumerator5
   )
   /SetVariables(Particles.MeshOrientation=&MultiplyQuaternion(
     Quaternion A=$Particles.SkeletalMeshLocation.SampledRotation
     Quaternion B=&MakeQuaternion(
      Quaternion From=3
      Axis=&SelectVectorFromArray(
       Array Sampling Mode=::NewEnumerator0
       Direct Array Index=?
       Vector Selection Array=DataInterface() {propertyValues {FloatData [<1 0 0>, <0 1 0>, <0 0 1>, <-1 0 0>, <0 -1 0>, <0 0 -1>]}}
       Random Seed=0
      )
      Angle=0
      From=<1 0 0>
      To=&SelectVectorFromArray(
       Array Sampling Mode=::NewEnumerator0
       Direct Array Index=?
       Vector Selection Array=DataInterface() {propertyValues {FloatData [<1 0 0>, <0 1 0>, <0 0 1>, <-1 0 0>, <0 -1 0>, <0 0 -1>]}}
       Random Seed=0
      )
      Stabilization Mode:ENiagara_OrientVectorCount=::NewEnumerator0
      Side Direction:Vector3f=?
      Up Direction:Vector3f=?
      Angle Type=::NewEnumerator0
      Normalize:NiagaraBool=true
     )
    )) {isSetParameters=true}
   &ScaleMeshSize(Scale Factor=&VectorFromFloat(Value=&FloatFromCurve(FloatCurve=DataInterface() {propertyValues {Curve {
         defaultValue=1
         keys [
          {arriveTangent=9.16, interpMode=RCIM_Cubic, leaveTangent=9.16, tangentMode=RCTM_User, time=0, value=0}
          {arriveTangent=1.6155, interpMode=RCIM_Cubic, leaveTangent=1.6155, tangentMode=RCTM_User, time=0.117, value=0.8038}
          {
           arriveTangent=0.0339
           arriveTangentWeight=0.5121
           interpMode=RCIM_Cubic
           leaveTangent=0.0339
           leaveTangentWeight=0.5121
           tangentMode=RCTM_User
           time=0.3149
           value=1
          }
          {arriveTangent=-1.9896, interpMode=RCIM_Cubic, leaveTangent=-1.9896, tangentMode=RCTM_User, time=0.8668, value=0.7089}
          {arriveTangent=-5.4864, interpMode=RCIM_Cubic, leaveTangent=-5.4864, tangentMode=RCTM_User, time=1, value=0}
         ]
        }}})))
   /SetVariables(Particles.VoxelParams:Vector4f=&MakeVector4(X=&RandomRangeFloat(Minimum=0.4, Maximum=0.8), Y=0, Z=0, W=0)) {isSetParameters=true, module=SetVariables_#8}
   /SetVariables(
    Particles.Glow_Base_CameraOffset=-5
    Particles.Glow_Base_Size:Vector2f=&Vector2DFromFloat(Value=&Multiply_Float(A=&MakeFloatFromVector(Vector=$Particles.Scale, Channel=::NewEnumerator0), B=1230))
   ) {isSetParameters=true, module=SetVariables_#9}
   &ScaleColor(
    Scale RGB=&VectorFromFloat(Value=&Lerp_Float(
      A=&RandomRangeFloat(Minimum=0.2, Maximum=3, Evaluation Type=::NewEnumerator1)
      B=1
      Alpha=&FloatFromCurve(FloatCurve=DataInterface() {propertyValues {Curve {defaultValue=1, keys [<at 0 0>, {tangentMode=RCTM_User, time=0.3536, value=0}, {tangentMode=RCTM_User, time=0.6446, value=1}, <at 1 1>]}}})
     ))
    ScaleA=true
    ScaleRGB=true
   )
   /SetVariables(
    Particles.JitterDirection:Vector3f=&RandomRangeVector(Minimum=<-3.5 -3.5 -3.5>, Maximum=<3.5 3.5 3.5>, Evaluation Type=::NewEnumerator1)
    Particles.JitterFallof:NiagaraFloat=&FloatFromCurve(
     FloatCurve=DataInterface() {propertyValues {Curve {
        defaultValue=0
        keys [
         {interpMode=RCIM_Cubic, tangentMode=RCTM_User, time=0, value=0}
         {arriveTangent=4.5081, interpMode=RCIM_Cubic, leaveTangent=4.5081, tangentMode=RCTM_User, time=0.392, value=0.5215}
         {interpMode=RCIM_Cubic, tangentMode=RCTM_User, time=0.5254, value=1}
         {arriveTangent=-2.1072, interpMode=RCIM_Cubic, leaveTangent=-2.1072, time=0.6702, value=0.2853}
         {tangentMode=RCTM_User, time=1, value=0}
        ]
       }}}
     Scale Curve=&RandomRangeFloat(Minimum=0.5, Maximum=1)
    )
    Particles.JitterStrength:NiagaraFloat=&FloatFromCurve(
     FloatCurve=DataInterface() {propertyValues {Curve {keys [<at 0 0 RCIM_Constant>,<at 0.8354 1 RCIM_Constant>]}}}
     CurveIndex=&RandomRangeFloat(Minimum=0, Maximum=1, Evaluation Type=::NewEnumerator1)
    )
   ) {isSetParameters=true, module=SetVariables_#10}
   /SetVariables(Particles.JitterPosition:NiagaraPosition=&Add_Vector(
     A=Linked() {linkedVariable {name=Particles.Position, type=NiagaraPosition}}
     B=&Multiply_VectorByFloat(
      Vector=$Particles.JitterDirection
      Float=&Multiply_Float(A=$Particles.JitterStrength, B=$Particles.JitterFallof)
     )
    )) {isSetParameters=true, module=SetVariables_#11}
   /SetVariables(
    Particles.GlitchColor:LinearColor=&MakeLinearColorFromVectorAndFloat(
     "Vector (RGB)"=&MakeVectorFromLinearColorRGB(Color=&ColorFromCurve(
       ColorCurve=DataInterface() {propertyValues {
         AlphaCurve {keys [<at 0 1 RCIM_Cubic>]}
         BlueCurve {keys [
           <at 0 0.7523>
           <at 0.1362 1>
           <at 0.3005 1>
           <at 0.4906 1>
           {arriveTangent=-2.3178, interpMode=RCIM_Cubic, leaveTangent=-2.3178, time=0.6656, value=1}
           <at 0.769 0.3547>
           <at 0.9227 0.1565>
           <at 1 0.1478 RCIM_Cubic>
          ]}
         GreenCurve {keys [
           <at 0 0.1632>
           <at 0.1362 0.0609>
           <at 0.3005 0.087>
           <at 0.4906 0.2463>
           {arriveTangent=2.7072, interpMode=RCIM_Cubic, leaveTangent=2.7072, time=0.6656, value=0.7595}
           <at 0.769 1>
           <at 0.9227 0.9528>
           <at 1 0.2044 RCIM_Cubic>
          ]}
         RedCurve {keys [
           <at 0 1>
           <at 0.1362 0.4789>
           <at 0.3005 0.3414>
           <at 0.4906 0.1019>
           {arriveTangent=0.4459, interpMode=RCIM_Cubic, leaveTangent=0.4459, time=0.6656, value=0}
           <at 0.769 0.2261>
           <at 0.9227 1>
           <at 1 1 RCIM_Cubic>
          ]}
        }}
       CurveIndex=&Frac_Float(Float=&Add_Float(
         A=&RandomRangeFloat(Minimum=0.7, Maximum=1.7)
         B=&RandomRangeFloat(Minimum=-0.7, Maximum=0.7, Evaluation Type=::NewEnumerator1)
        ))
      ))
     "Float (Alpha)"=0.6
    )
    Particles.GlitchPosition:NiagaraPosition=&Add_Vector(
     A=Linked() {linkedVariable {name=Particles.Position, type=NiagaraPosition}}
     B=&Multiply_VectorByFloat(
      Vector=$Particles.JitterDirection
      Float=&Multiply_Float(A=&Multiply_Float(A=$Particles.JitterStrength, B=$Particles.JitterFallof), B=-2)
     )
    )
    Particles.GlitchScale:Vector3f=&Multiply_VectorByFloat(Vector=$Particles.Scale, Float=$Particles.JitterStrength)
   ) {isSetParameters=true, module=SetVariables_#12}
  ]
  rendererClasses [&NiagaraMeshRendererProperties, &NiagaraSpriteRendererProperties]
  renderers [
   {rendererClass=&NiagaraMeshRendererProperties, rendererIndex=0}
   {rendererClass=&NiagaraSpriteRendererProperties, rendererIndex=1}
   {rendererClass=&NiagaraMeshRendererProperties, rendererIndex=2}
  ]
 }
}
summary {
 systemName=NS_Glider_Season_42_Idle
 emitters [
  {
   emitterName=Glow
   rendererClasses [
    &NiagaraSpriteRendererProperties
    &NiagaraLightRendererProperties
    &NiagaraMeshRendererProperties
   ]
  }
  {emitterName=CrownCubes, rendererClasses [&NiagaraMeshRendererProperties, &NiagaraSpriteRendererProperties]}
  {emitterName=Glitch_Minor, rendererClasses [&NiagaraSpriteRendererProperties]}
  {emitterName=Glitch_Major, rendererClasses [&NiagaraSpriteRendererProperties]}
  {emitterName=Glitch_Blocks, rendererClasses [&NiagaraMeshRendererProperties]}
  {emitterName=TrailCubes, rendererClasses [&NiagaraMeshRendererProperties, &NiagaraSpriteRendererProperties]}
  {
   bEnabled=false
   emitterName=Glow001
   rendererClasses [
    &NiagaraSpriteRendererProperties
    &NiagaraLightRendererProperties
    &NiagaraMeshRendererProperties
   ]
  }
 ]
 userVariables [
  User.bIsFrontEnd {type=NiagaraBool}
  User.bIsFrontEndPreview {type=NiagaraBool}
  User.bisHeadHidden {type=NiagaraBool}
  User.ForwardDot {type=NiagaraFloat, defaultValue=1, description=""}
  User.RightDot {type=NiagaraFloat, defaultValue=1, description=""}
  User.Variant {type=NiagaraFloat, defaultValue=1, description=""}
 ]
}
