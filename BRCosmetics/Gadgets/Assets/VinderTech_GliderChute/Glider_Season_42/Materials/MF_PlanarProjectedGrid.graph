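# Material function that produces an animated grid on a planar 2D projection, stored with the Season 42 VinderTech glider assets.
# Rotates the supplied Position Plane coordinates by Rotation, then adds Offset to form the UVs for the shared MF_AnimatedGrid.
# Tiling and Speed pass through to that function, while its color and line settings stay unconnected.
# Outputs the grayscale grid and the grid lines separately.
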MaterialFunction /BRCosmetics/Gadgets/Assets/VinderTech_GliderChute/Glider_Season_42/Materials/MF_PlanarProjectedGrid
graph {nodes [
  n1 = FunctionInput() {inputName=Position Plane, inputType=FunctionInput_Vector2, sortPriority=10}
  n2 = FunctionInput() {bUsePreviewValueAsDefault=true, inputName=Rotation, inputType=FunctionInput_Scalar, sortPriority=20}
  n3 = FunctionInput() {bUsePreviewValueAsDefault=true, inputName=Tiling, inputType=FunctionInput_Scalar, sortPriority=30, previewValue=<0.5 0 0 1>}
  n4 = FunctionInput() {bUsePreviewValueAsDefault=true, inputName=Speed, inputType=FunctionInput_Scalar, sortPriority=40, previewValue=<0.1 0 0 1>}
  n5 = MaterialFunctionCall!(
   UV=Add(
    A=FunctionInput() {bUsePreviewValueAsDefault=true, inputName=Offset, inputType=FunctionInput_Vector2, sortPriority=11}
    B=Rotator(Coordinate=@n1, Time=@n2)
   )
   Tiling=@n3
   Tiling X
   Tiling Y
   Color 01
   Color 02
   Speed=@n4
   Speed Random Amount
   Line Thickness
   Line Hardness
   Line Color
  ) {materialFunction=/Game/Global/MaterialFunctions/Cosmetics/AnimatedGrid/MF_AnimatedGrid}
  FunctionOutput(None=@n5.Grayscale Grid) {outputName=Grayscale Grid}
  FunctionOutput(None=@n5.Gird Lines) {outputName=Grid Lines, sortPriority=10}
 ]}
